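# Idle glow effect for the PsychicAphid exo character body, drawing persistent sprite glows on its skeletal mesh.
# Four emitters (center, shoulders, head, large glows) burst-spawn particles whose partition index selects bone, offset, size and bone-rotated capsule alignment from hand-authored arrays.
# A shared system color, modulated by layered per-particle waveforms, tints the capsule, soft-glow and fake-light attributes.
# User head/hand-hidden flags drive visibility tags on some glows.
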NiagaraSystem /BRCosmetics/Characters/Player/Male/Medium/Bodies/M_MED_PsychicAphid_Exo/FX/NS_PsychicAphid_Exo_Idle
dependencies {
 usedDataInterfaces [
  &NiagaraDataInterfaceArrayFloat
  &NiagaraDataInterfaceArrayInt32
  &NiagaraDataInterfaceArrayFloat3
  &NiagaraDataInterfacePlatformSet
  &NiagaraDataInterfaceSkeletalMesh
 ]
 usedDynamicInputs [
  &SimulationPosition
  &FNTimeOfDayScalar
  &FN_PickFloatDependOnPlatform
  &FN_FrontEndFloatModifier
 ]
 usedModules [
  &self:SystemSpawnScript.SystemScriptSource.SystemScriptGraph.NiagaraNodeAssignment_3.SetVariables_#1
  &self:SystemSpawnScript.SystemScriptSource.SystemScriptGraph.NiagaraNodeAssignment_1.SetVariables_#2
  &self:SystemSpawnScript.SystemScriptSource.SystemScriptGraph.NiagaraNodeAssignment_0.SetVariables_#3
  &SystemState
 ]
}
emitters {
 BigGlows {
  emitterUpdateScript [
   &EmitterState(
    Inactive Response=::NewEnumerator1
    Life Cycle Mode=::NewEnumerator1
    Loop Behavior=::NewEnumerator1
    Loop Duration Mode=::NewEnumerator1
    UseLoopDelay=false
   )
   &SpawnBurst_Instantaneous(Spawn Count=8, Use Spawn Probability=false, UseLoopCountLimit=false)
  ]
  particleSpawnScript [
   &PartitionParticles(Number Of Partitions=8, Particles Per Partition=1, Partition Mode:ENiagara_PartitionMode=::NewEnumerator1)
   &InitializeParticle(Color Mode=::NewEnumerator0, UsePositionOffset=false, Write Material Random=true)
  ]
  particleUpdateScript [
   &ParticleState(Kill Particles When Lifetime Has Elapsed=false, Let Infinitely Lived Particles Die When Emitter Deactivates=true)
   /SetVariables(Particles.MainColor:LinearColor=&MultiplyLinearColor(
     Linear Color=&MultiplyLinearColor(
      Linear Color=$System.MainColor
      Scale Factor=&MakeFloat4FromVectorAndFloat("Vector (RGB)"=<1 1 1>, "Float (Alpha)"=0.7)
     )
     Scale Factor=&MakeVector4_FromFloat(Float=&Waveform(
       Waveform Count=::NewEnumerator1
       Global Input Position=&ReturnNormalizedExecIndex(Normalized Index Scale=9)
       "Waveform [1]"=::NewEnumerator12
       "[1] Frequency"=2.65
       "[1] Amplitude Min/Max"=<0.7 1>
       "Waveform [2] Blend"=::NewEnumerator5
       "[2] Interpolation"=0.5
       "[2] Invert Interpolation"=false
       "Waveform [2]"=::NewEnumerator12
       "[2] Input Position"=0
       "[2] Phase"=$Emitter.Age
       "[2] Phase Scale"=1
       "[2] Compound Frequency"=2
       "[2] Frequency"=16
       "[2] Amplitude Min/Max"=<0.75 1>
       "[2] Scale"=1
       "[2] Bias"=0
       "[2] Exponent"=1
       "[2] Pulse Duration"=0.1
       "[2] Clip Triangle Min/Max"=<0 0>
       "[2] Random Seed"=0
      ))
    )) {isSetParameters=true, module=SetVariables_#4}
   /SetVariables(
    Particles.CapsuleGlowColor:LinearColor=&MultiplyLinearColor(
     Linear Color=$StackContext.MainColor
     Scale Factor=&MakeFloat4FromVectorAndFloat("Vector (RGB)"=&VectorFromFloat(Value=1.5), "Float (Alpha)"=1)
    )
    Particles.CapsuleSpriteSize:Vector2f=&Vector2DFromFloat(Value=&SelectFloatFromArray(
      Direct Array Index=$Particles.PartitionParticles.Partition
      Float Selection Array=DataInterface() {propertyValues {FloatData [15,15,15,15,10,10,10,10]}}
     ))
   ) {isSetParameters=true, module=SetVariables_#5}
   &SkeletalMeshLocation(
    Bone / Socket Index=&SelectIntFromArray(
     Direct Array Index=$Particles.PartitionParticles.Partition
     Int Selection Array=DataInterface() {propertyValues {IntData [8,10,3,3,7,7,9,9]}}
    )
    Bone Sampling Mode=::NewEnumerator4
    MaskSpawnGroup=false
    Sampled Position Offset=&SelectVectorFromArray(
     Direct Array Index=$Particles.PartitionParticles.Partition
     Vector Selection Array=DataInterface() {propertyValues {FloatData [<-10.9594 0.1956 -5.9162>, <10.9594 -0.1956 5.9162>, <3 9 7>, <3 9 -7>, <-4.0545 5 -2.8358>, <-4.5 -3 -2.8>, <4.0545 -5 2.8358>, <4.5 3 2.8>]}}
    )
    Skeletal Mesh=$System.SkeletalMesh
   )
   /SetVariables(Particles.CapsuleSpriteAlignment:Vector3f=&RotateVectorByQuaternion(
     Vector To Rotate=&SelectVectorFromArray(
      Direct Array Index=$Particles.PartitionParticles.Partition
      Vector Selection Array=DataInterface() {propertyValues {FloatData [<1 0 -0.2>, <1 0 -0.25>, <1 0.2 0>, <1 0.2 0>, <0 0.3 1>, <0.3 -0.3 1>, <0 0.3 1>, <0 -0.3 1>]}}
     )
     Quaternion=$Particles.SkeletalMeshLocation.SampledRotation
    )) {isSetParameters=true, module=SetVariables_#6}
   /SetVariables(Particles.CapsuleSpriteDynamic:Vector4f=&MakeFloat4FromVectorAndFloat(
     "Vector (RGB)"=&NormalizeVector(Vector To Normalize=$Particles.CapsuleSpriteAlignment, Safe Normalize=true, Safe Fallback Threshold=0, Safe Fallback Vector=<0 0 0.001>)
     "Float (Alpha)"=&SelectFloatFromArray(
      Direct Array Index=$Particles.PartitionParticles.Partition
      Float Selection Array=DataInterface() {propertyValues {FloatData [30,30,30,30,15,15,15,15]}}
     )
    )) {isSetParameters=true, module=SetVariables_#7}
   /SetVariables(Particles.VisibilityTag=&MakeCustomIntFromBool(
     Bool=&Boolean_OrOperation(
      Bool1=&SetBoolByIntComparison(A=$Particles.PartitionParticles.Partition, B=2, Comparison Type=::NewEnumerator6)
      Bool2=&SetBoolByIntComparison(A=$Particles.PartitionParticles.Partition, B=3, Comparison Type=::NewEnumerator0)
     )
     True INT=&MakeCustomIntFromBool(Bool=$User.bisHandHidden, True INT=1, False INT=0)
     False INT=0
    )) {isSetParameters=true}
  ]
  rendererClasses [&NiagaraSpriteRendererProperties]
  renderers [{rendererClass=&NiagaraSpriteRendererProperties}]
 }
 Glow_Center {
  emitterUpdateScript [
   &EmitterState(
    Inactive Response=::NewEnumerator1
    Life Cycle Mode=::NewEnumerator1
    Loop Behavior=::NewEnumerator1
    Loop Duration Mode=::NewEnumerator1
    UseLoopDelay=false
   )
   &SpawnBurst_Instantaneous(Spawn Count=15, Use Spawn Probability=false, UseLoopCountLimit=false)
  ]
  particleSpawnScript [
   &PartitionParticles(Number Of Partitions=15, Particles Per Partition=1, Partition Mode:ENiagara_PartitionMode=::NewEnumerator1)
   &InitializeParticle(Color Mode=::NewEnumerator0, UsePositionOffset=false, Write Material Random=true)
  ]
  particleUpdateScript [
   &ParticleState(Kill Particles When Lifetime Has Elapsed=false, Let Infinitely Lived Particles Die When Emitter Deactivates=true)
   /SetVariables(Particles.MainColor:LinearColor=&MultiplyLinearColor(
     Linear Color=&MultiplyLinearColor(
      Linear Color=$System.MainColor
      Scale Factor=&MakeFloat4FromVectorAndFloat("Vector (RGB)"=<1 1 1>, "Float (Alpha)"=&Multiply_Float(A=1, B=1))
     )
     Scale Factor=&MakeVector4_FromFloat(Float=&Waveform(
       Waveform Count=::NewEnumerator1
       Global Input Position=&ReturnNormalizedExecIndex(Normalized Index Scale=9)
       "Waveform [1]"=::NewEnumerator12
       "[1] Frequency"=2.65
       "[1] Amplitude Min/Max"=<0.7 1>
       "Waveform [2] Blend"=::NewEnumerator5
       "[2] Interpolation"=0.5
       "[2] Invert Interpolation"=false
       "Waveform [2]"=::NewEnumerator12
       "[2] Input Position"=0
       "[2] Phase"=$Emitter.Age
       "[2] Phase Scale"=1
       "[2] Compound Frequency"=2
       "[2] Frequency"=16
       "[2] Amplitude Min/Max"=<0.75 1>
       "[2] Scale"=1
       "[2] Bias"=0
       "[2] Exponent"=1
       "[2] Pulse Duration"=0.1
       "[2] Clip Triangle Min/Max"=<0 0>
       "[2] Random Seed"=0
      ))
    )) {isSetParameters=true, module=SetVariables_#4}
   /SetVariables(
    Particles.SoftGlowCameraOffset:NiagaraFloat=&SelectFloatFromArray(
     Direct Array Index=$Particles.PartitionParticles.Partition
     Float Selection Array=DataInterface() {propertyValues {FloatData [3,3,2,4,4,4,2,4,4,4,4,4,4,4,4]}}
    )
    Particles.SoftGlowColor:LinearColor=&MultiplyLinearColor(
     Linear Color=$StackContext.MainColor
     Scale Factor=&MakeFloat4FromVectorAndFloat(
      "Vector (RGB)"=&VectorFromFloat(Value=&SelectFloatFromArray(
        Direct Array Index=$Particles.PartitionParticles.Partition
        Float Selection Array=DataInterface() {propertyValues {FloatData [5,5,5,5,5,5,5,5,5,5,5,5,5,5,5]}}
       ))
      "Float (Alpha)"=0.3
     )
    )
    Particles.SoftGlowSize:Vector2f=&Vector2DFromFloat(Value=&SelectFloatFromArray(
      Direct Array Index=$Particles.PartitionParticles.Partition
      Float Selection Array=DataInterface() {propertyValues {FloatData [17,17,10,10,15,15,11,11,11,11,11,11,11,11,11]}}
     ))
   ) {isSetParameters=true, module=SetVariables_#8}
   /SetVariables(
    Particles.FakeLightColor:LinearColor=&MultiplyLinearColor(
     Linear Color=$StackContext.MainColor
     Scale Factor=&MakeFloat4FromVectorAndFloat(
      "Vector (RGB)"=&VectorFromFloat(Value=&SelectFloatFromArray(
        Direct Array Index=$Particles.PartitionParticles.Partition
        Float Selection Array=DataInterface() {propertyValues {FloatData [3,3,3,3,2,3,3,3,3,3,3,3,3,3,3]}}
       ))
      "Float (Alpha)"=1
     )
    )
    Particles.FakeLightSize:Vector2f=&Vector2DFromFloat(Value=&SelectFloatFromArray(
      Direct Array Index=$Particles.PartitionParticles.Partition
      Float Selection Array=DataInterface() {propertyValues {FloatData [17,17,10,10,20,20,12,12,12,12,12,12,8,8,8]}}
     ))
   ) {isSetParameters=true, module=SetVariables_#9}
   /SetVariables(
    Particles.CapsuleGlowColor:LinearColor=&MultiplyLinearColor(
     Linear Color=$StackContext.MainColor
     Scale Factor=&MakeFloat4FromVectorAndFloat(
      "Vector (RGB)"=&VectorFromFloat(Value=&SelectFloatFromArray(
        Direct Array Index=$Particles.PartitionParticles.Partition
        Float Selection Array=DataInterface() {propertyValues {FloatData [2,2,2,2,2,2,2,2,2,2,2,2,2,2,2]}}
       ))
      "Float (Alpha)"=1
     )
    )
    Particles.CapsuleSpriteSize:Vector2f=&Vector2DFromFloat(Value=&SelectFloatFromArray(
      Direct Array Index=$Particles.PartitionParticles.Partition
      Float Selection Array=DataInterface() {propertyValues {FloatData [13,13,8,8,15,15,7,7,7,7,7,7,7,8,8]}}
     ))
   ) {isSetParameters=true, module=SetVariables_#5}
   &SkeletalMeshLocation(
    Bone / Socket Index=&SelectIntFromArray(
     Direct Array Index=$Particles.PartitionParticles.Partition
     Int Selection Array=DataInterface() {propertyValues {IntData [0,0,1,1,0,0,2,2,3,3,4,4,0,5,6]}}
    )
    Bone Sampling Mode=::NewEnumerator4
    MaskSpawnGroup=false
    Sampled Position Offset=&SelectVectorFromArray(
     Direct Array Index=$Particles.PartitionParticles.Partition
     Vector Selection Array=DataInterface() {propertyValues {FloatData [
        <8 -10 -8>
        <8 -10 8>
        <-2 -11 -5.5>
        <-2 -11 5>
        <-3.4102 11 7.7035>
        <-3.4102 11 -7>
        <2.1433 9 8.8>
        <2.1433 9 -8.8>
        <2.1433 8 7.5>
        <2.1433 8 -7.5>
        <2.1433 7 7.5>
        <2.1433 7 -7.5>
        <-5 11 0>
        <25.319 -7.8547 -4.9686>
        <-25.319 7.8547 4.9686>
       ]}}
    )
    Skeletal Mesh=$System.SkeletalMesh
   )
   /SetVariables(Particles.CapsuleSpriteAlignment:Vector3f=&RotateVectorByQuaternion(
     Vector To Rotate=&SelectVectorFromArray(
      Direct Array Index=$Particles.PartitionParticles.Partition
      Vector Selection Array=DataInterface() {propertyValues {FloatData [
         <-1.1 -0.5586 1>
         <1.1 0.5586 1>
         <1 0 0.332>
         <1 0 -0.332>
         <1 0 0.5>
         <1 0 -0.5>
         <1 -0.3 0.5>
         <1 -0.3 -0.5>
         <1 -0.3 0.5>
         <1 -0.3 -0.5>
         <1 -0.3 0.5>
         <1 -0.3 -0.5>
         <1 0 0>
         <1 0.3 0>
         <1 0.3 0>
        ]}}
     )
     Quaternion=$Particles.SkeletalMeshLocation.SampledRotation
    )) {isSetParameters=true, module=SetVariables_#6}
   /SetVariables(Particles.CapsuleSpriteDynamic:Vector4f=&MakeFloat4FromVectorAndFloat(
     "Vector (RGB)"=&NormalizeVector(Vector To Normalize=$Particles.CapsuleSpriteAlignment, Safe Normalize=true, Safe Fallback Threshold=0, Safe Fallback Vector=<0 0 0.001>)
     "Float (Alpha)"=&SelectFloatFromArray(
      Direct Array Index=$Particles.PartitionParticles.Partition
      Float Selection Array=DataInterface() {propertyValues {FloatData [24,24,13,13,18,19,12,12,12,12,12,12,9,12,12]}}
     )
    )) {isSetParameters=true, module=SetVariables_#7}
  ]
  rendererClasses [&NiagaraSpriteRendererProperties]
  renderers [
   {rendererClass=&NiagaraSpriteRendererProperties}
   {rendererClass=&NiagaraSpriteRendererProperties, rendererIndex=1}
   {rendererClass=&NiagaraSpriteRendererProperties, rendererIndex=2}
  ]
 }
 Glow_Head {
  emitterUpdateScript [
   &EmitterState(
    Inactive Response=::NewEnumerator1
    Life Cycle Mode=::NewEnumerator1
    Loop Behavior=::NewEnumerator1
    Loop Duration Mode=::NewEnumerator1
    UseLoopDelay=false
   )
   &SpawnBurst_Instantaneous(Spawn Count=8, Use Spawn Probability=false, UseLoopCountLimit=false)
   &SpawnBurst_Instantaneous(Spawn Count=8, Spawn Time=0.4227, Use Spawn Probability=false, UseLoopCountLimit=false) {module=SpawnBurst_Instantaneous001}
  ]
  particleSpawnScript [
   &PartitionParticles(Number Of Partitions=8, Particles Per Partition=1, Partition Mode:ENiagara_PartitionMode=::NewEnumerator1)
   &InitializeParticle(Color Mode=::NewEnumerator0, UsePositionOffset=false, Write Material Random=true)
  ]
  particleUpdateScript [
   &ParticleState(Kill Particles When Lifetime Has Elapsed=false, Let Infinitely Lived Particles Die When Emitter Deactivates=true)
   /SetVariables(Particles.MainColor:LinearColor=&MultiplyLinearColor(
     Linear Color=&MultiplyLinearColor(
      Linear Color=$System.MainColor
      Scale Factor=&MakeFloat4FromVectorAndFloat("Vector (RGB)"=<1 1 1>, "Float (Alpha)"=&Multiply_Float(A=1, B=1))
     )
     Scale Factor=&MakeVector4_FromFloat(Float=&Waveform(
       Waveform Count=::NewEnumerator1
       Global Input Position=&ReturnNormalizedExecIndex(Normalized Index Scale=9)
       "Waveform [1]"=::NewEnumerator12
       "[1] Frequency"=2.65
       "[1] Amplitude Min/Max"=<0.7 1>
       "Waveform [2] Blend"=::NewEnumerator5
       "[2] Interpolation"=0.5
       "[2] Invert Interpolation"=false
       "Waveform [2]"=::NewEnumerator12
       "[2] Input Position"=0
       "[2] Phase"=$Emitter.Age
       "[2] Phase Scale"=1
       "[2] Compound Frequency"=2
       "[2] Frequency"=16
       "[2] Amplitude Min/Max"=<0.75 1>
       "[2] Scale"=1
       "[2] Bias"=0
       "[2] Exponent"=1
       "[2] Pulse Duration"=0.1
       "[2] Clip Triangle Min/Max"=<0 0>
       "[2] Random Seed"=0
      ))
    )) {isSetParameters=true, module=SetVariables_#4}
   /SetVariables(
    Particles.SoftGlowCameraOffset:NiagaraFloat=&SelectFloatFromArray(
     Direct Array Index=$Particles.PartitionParticles.Partition
     Float Selection Array=DataInterface() {propertyValues {FloatData [3,3,3,3,3,3,3,3]}}
    )
    Particles.SoftGlowColor:LinearColor=&MultiplyLinearColor(
     Linear Color=$StackContext.MainColor
     Scale Factor=&MakeFloat4FromVectorAndFloat(
      "Vector (RGB)"=&VectorFromFloat(Value=&SelectFloatFromArray(
        Direct Array Index=$Particles.PartitionParticles.Partition
        Float Selection Array=DataInterface() {propertyValues {FloatData [5,5,5,5,5,5,5,5]}}
       ))
      "Float (Alpha)"=0.3
     )
    )
    Particles.SoftGlowSize:Vector2f=&Vector2DFromFloat(Value=&SelectFloatFromArray(
      Direct Array Index=$Particles.PartitionParticles.Partition
      Float Selection Array=DataInterface() {propertyValues {FloatData [5,5,5,5,5,5,5,5]}}
     ))
   ) {isSetParameters=true, module=SetVariables_#8}
   /SetVariables(
    Particles.FakeLightColor:LinearColor=&MultiplyLinearColor(
     Linear Color=$StackContext.MainColor
     Scale Factor=&MakeFloat4FromVectorAndFloat(
      "Vector (RGB)"=&VectorFromFloat(Value=&SelectFloatFromArray(
        Direct Array Index=$Particles.PartitionParticles.Partition
        Float Selection Array=DataInterface() {propertyValues {FloatData [3,3,3,3,2,3,3,3]}}
       ))
      "Float (Alpha)"=1
     )
    )
    Particles.FakeLightSize:Vector2f=&Vector2DFromFloat(Value=&SelectFloatFromArray(
      Direct Array Index=$Particles.PartitionParticles.Partition
      Float Selection Array=DataInterface() {propertyValues {FloatData [6,6,6,6,4,4,4,4]}}
     ))
   ) {isSetParameters=true, module=SetVariables_#9}
   /SetVariables(
    Particles.CapsuleGlowColor:LinearColor=&MultiplyLinearColor(
     Linear Color=$StackContext.MainColor
     Scale Factor=&MakeFloat4FromVectorAndFloat(
      "Vector (RGB)"=&VectorFromFloat(Value=&SelectFloatFromArray(
        Direct Array Index=$Particles.PartitionParticles.Partition
        Float Selection Array=DataInterface() {propertyValues {FloatData [2,2,2,2,2,2,2,2]}}
       ))
      "Float (Alpha)"=1
     )
    )
    Particles.CapsuleSpriteSize:Vector2f=&Vector2DFromFloat(Value=&SelectFloatFromArray(
      Direct Array Index=$Particles.PartitionParticles.Partition
      Float Selection Array=DataInterface() {propertyValues {FloatData [5,6,5,6,3,3,3,3]}}
     ))
   ) {isSetParameters=true, module=SetVariables_#5}
   &SkeletalMeshLocation(
    Bone / Socket Index=&SelectIntFromArray(
     Direct Array Index=$Particles.PartitionParticles.Partition
     Int Selection Array=DataInterface() {propertyValues {IntData [11,11,11,11,11,11,11,11]}}
    )
    Bone Sampling Mode=::NewEnumerator4
    MaskSpawnGroup=false
    Sampled Position Offset=&SelectVectorFromArray(
     Direct Array Index=$Particles.PartitionParticles.Partition
     Vector Selection Array=DataInterface() {propertyValues {FloatData [
        <7.0146 10 -4>
        <9.4311 8.5473 -4.9711>
        <7.0146 10 4>
        <9.4311 8.5473 4.9711>
        <4.3721 -5.3323 -5.3328>
        <5.2999 -4.6968 -6.2165>
        <4.3721 -5.3323 5.3328>
        <5.2999 -4.6968 6.2165>
       ]}}
    )
    Skeletal Mesh=$System.SkeletalMesh
   )
   /SetVariables(Particles.CapsuleSpriteAlignment:Vector3f=&RotateVectorByQuaternion(
     Vector To Rotate=&SelectVectorFromArray(
      Direct Array Index=$Particles.PartitionParticles.Partition
      Vector Selection Array=DataInterface() {propertyValues {FloatData [<-0.6 0.7 1>, <-0.8 1 1>, <-0.6 0.7 -1>, <-0.8 1 -1>, <0.8 -1 1>, <0.8 -1 1>, <1.2 -1 -1>, <1.2 -1 -1>]}}
     )
     Quaternion=$Particles.SkeletalMeshLocation.SampledRotation
    )) {isSetParameters=true, module=SetVariables_#6}
   /SetVariables(Particles.CapsuleSpriteDynamic:Vector4f=&MakeFloat4FromVectorAndFloat(
     "Vector (RGB)"=&NormalizeVector(Vector To Normalize=$Particles.CapsuleSpriteAlignment, Safe Normalize=true, Safe Fallback Threshold=0, Safe Fallback Vector=<0 0 0.001>)
     "Float (Alpha)"=&SelectFloatFromArray(
      Direct Array Index=$Particles.PartitionParticles.Partition
      Float Selection Array=DataInterface() {propertyValues {FloatData [10,10,10,10,5,5,5,5]}}
     )
    )) {isSetParameters=true, module=SetVariables_#7}
   /SetVariables(Particles.VisibilityTag=&MakeCustomIntFromBool(Bool=$User.bisHeadHidden, True INT=1, False INT=0)) {isSetParameters=true}
  ]
  rendererClasses [&NiagaraSpriteRendererProperties]
  renderers [
   {rendererClass=&NiagaraSpriteRendererProperties}
   {rendererClass=&NiagaraSpriteRendererProperties, rendererIndex=1}
   {rendererClass=&NiagaraSpriteRendererProperties, rendererIndex=2}
  ]
 }
 Glow_Shoulders {
  emitterUpdateScript [
   &EmitterState(
    Inactive Response=::NewEnumerator1
    Life Cycle Mode=::NewEnumerator1
    Loop Behavior=::NewEnumerator1
    Loop Duration Mode=::NewEnumerator1
    UseLoopDelay=false
   )
   &SpawnBurst_Instantaneous(Spawn Count=13, Use Spawn Probability=false, UseLoopCountLimit=false)
  ]
  particleSpawnScript [
   &PartitionParticles(Number Of Partitions=13, Particles Per Partition=1, Partition Mode:ENiagara_PartitionMode=::NewEnumerator1)
   &InitializeParticle(Color Mode=::NewEnumerator0, UsePositionOffset=false, Write Material Random=true)
  ]
  particleUpdateScript [
   &ParticleState(Kill Particles When Lifetime Has Elapsed=false, Let Infinitely Lived Particles Die When Emitter Deactivates=true)
   /SetVariables(Particles.MainColor:LinearColor=&MultiplyLinearColor(
     Linear Color=&MultiplyLinearColor(
      Linear Color=$System.MainColor
      Scale Factor=&MakeFloat4FromVectorAndFloat(
       "Vector (RGB)"=<1 1 1>
       "Float (Alpha)"=&Multiply_Float(
        A=&SelectFloatFromArray(
         Direct Array Index=$Particles.PartitionParticles.Partition
         Float Selection Array=DataInterface() {propertyValues {FloatData [1,1,1,1,1,1,1,1,1,1,1,1,1]}}
        )
        B=1
       )
      )
     )
     Scale Factor=&MakeVector4_FromFloat(Float=&Waveform(
       Waveform Count=::NewEnumerator1
       Global Input Position=&ReturnNormalizedExecIndex(Normalized Index Scale=9)
       "Waveform [1]"=::NewEnumerator12
       "[1] Frequency"=2.65
       "[1] Amplitude Min/Max"=<0.7 1>
       "Waveform [2] Blend"=::NewEnumerator5
       "[2] Interpolation"=0.5
       "[2] Invert Interpolation"=false
       "Waveform [2]"=::NewEnumerator12
       "[2] Input Position"=0
       "[2] Phase"=$Emitter.Age
       "[2] Phase Scale"=1
       "[2] Compound Frequency"=2
       "[2] Frequency"=16
       "[2] Amplitude Min/Max"=<0.75 1>
       "[2] Scale"=1
       "[2] Bias"=0
       "[2] Exponent"=1
       "[2] Pulse Duration"=0.1
       "[2] Clip Triangle Min/Max"=<0 0>
       "[2] Random Seed"=0
      ))
    )) {isSetParameters=true, module=SetVariables_#4}
   /SetVariables(
    Particles.SoftGlowCameraOffset:NiagaraFloat=&SelectFloatFromArray(
     Direct Array Index=$Particles.PartitionParticles.Partition
     Float Selection Array=DataInterface() {propertyValues {FloatData [3,3,3,3,3,3,8,3,3,3,3,3,3]}}
    )
    Particles.SoftGlowColor:LinearColor=&MultiplyLinearColor(
     Linear Color=$StackContext.MainColor
     Scale Factor=&MakeFloat4FromVectorAndFloat(
      "Vector (RGB)"=&VectorFromFloat(Value=&SelectFloatFromArray(
        Direct Array Index=$Particles.PartitionParticles.Partition
        Float Selection Array=DataInterface() {propertyValues {FloatData [5,5,5,5,5,5,5,5,5,5,5,5,5]}}
       ))
      "Float (Alpha)"=0.3
     )
    )
    Particles.SoftGlowSize:Vector2f=&Vector2DFromFloat(Value=&SelectFloatFromArray(
      Direct Array Index=$Particles.PartitionParticles.Partition
      Float Selection Array=DataInterface() {propertyValues {FloatData [8,8,8,8,8,8,11,11,11,11,11,11,11]}}
     ))
   ) {isSetParameters=true, module=SetVariables_#8}
   /SetVariables(
    Particles.FakeLightColor:LinearColor=&MultiplyLinearColor(
     Linear Color=$StackContext.MainColor
     Scale Factor=&MakeFloat4FromVectorAndFloat(
      "Vector (RGB)"=&VectorFromFloat(Value=&SelectFloatFromArray(
        Direct Array Index=$Particles.PartitionParticles.Partition
        Float Selection Array=DataInterface() {propertyValues {FloatData [3,3,3,3,3,3,3,3,3,3,3,3,3]}}
       ))
      "Float (Alpha)"=1
     )
    )
    Particles.FakeLightSize:Vector2f=&Vector2DFromFloat(Value=&SelectFloatFromArray(
      Direct Array Index=$Particles.PartitionParticles.Partition
      Float Selection Array=DataInterface() {propertyValues {FloatData [8,8,8,8,8,8,15,8,8,8,8,8,8]}}
     ))
   ) {isSetParameters=true, module=SetVariables_#9}
   /SetVariables(
    Particles.CapsuleGlowColor:LinearColor=&MultiplyLinearColor(
     Linear Color=$StackContext.MainColor
     Scale Factor=&MakeFloat4FromVectorAndFloat(
      "Vector (RGB)"=&VectorFromFloat(Value=&SelectFloatFromArray(
        Direct Array Index=$Particles.PartitionParticles.Partition
        Float Selection Array=DataInterface() {propertyValues {FloatData [2,2,2,2,2,2,2,2,2,2,2,2,2]}}
       ))
      "Float (Alpha)"=1
     )
    )
    Particles.CapsuleSpriteSize:Vector2f=&Vector2DFromFloat(Value=&SelectFloatFromArray(
      Direct Array Index=$Particles.PartitionParticles.Partition
      Float Selection Array=DataInterface() {propertyValues {FloatData [6,5,6,6,6,6,8,6,6,6,6,6,6]}}
     ))
   ) {isSetParameters=true, module=SetVariables_#5}
   &SkeletalMeshLocation(
    Bone / Socket Index=&SelectIntFromArray(
     Direct Array Index=$Particles.PartitionParticles.Partition
     Int Selection Array=DataInterface() {propertyValues {IntData [7,7,7,7,7,7,8,8,8,8,8,8,8]}}
    )
    Bone Sampling Mode=::NewEnumerator4
    MaskSpawnGroup=false
    Sampled Position Offset=&SelectVectorFromArray(
     Direct Array Index=$Particles.PartitionParticles.Partition
     Vector Selection Array=DataInterface() {propertyValues {FloatData [
        <-5.3215 -5.9656 -1.2765>
        <-4.9624 -5.165 -3.6343>
        <-4.6023 -3.251 -5.9909>
        <-4.0204 4.6868 -5.6085>
        <-4.0709 6.5539 -2.8289>
        <-3.593 7.6254 -0.027>
        <-6.0451 0.1284 -7.2874>
        <-8.2132 -3.2604 -4.734>
        <-12.6807 -2.3974 -4.3457>
        <-17.0918 -2.1056 -3.5774>
        <-8.1062 3.2013 -4.4314>
        <-12.5112 2.562 -3.9633>
        <-17.0275 2.3702 -3.2917>
       ]}}
    )
    Skeletal Mesh=$System.SkeletalMesh
   )
   /SetVariables(Particles.CapsuleSpriteAlignment:Vector3f=&RotateVectorByQuaternion(
     Vector To Rotate=&SelectVectorFromArray(
      Direct Array Index=$Particles.PartitionParticles.Partition
      Vector Selection Array=DataInterface() {propertyValues {FloatData [
         <1 -0.35 0.4>
         <1 -0.35 0.2>
         <1 -0.2 -0.1>
         <1 0.3 0>
         <1 0.35 0.1>
         <1 0.2 0.2>
         <1 0 -0.35>
         <1 -0.4 0.3>
         <1 -0.45 0>
         <1 -0.45 0>
         <1 0.45 0.3>
         <1 0.45 -0.4>
         <1 0.45 0>
        ]}}
     )
     Quaternion=$Particles.SkeletalMeshLocation.SampledRotation
    )) {isSetParameters=true, module=SetVariables_#6}
   /SetVariables(Particles.CapsuleSpriteDynamic:Vector4f=&MakeFloat4FromVectorAndFloat(
     "Vector (RGB)"=&NormalizeVector(Vector To Normalize=$Particles.CapsuleSpriteAlignment, Safe Normalize=true, Safe Fallback Threshold=0, Safe Fallback Vector=<0 0 0.001>)
     "Float (Alpha)"=&SelectFloatFromArray(
      Direct Array Index=$Particles.PartitionParticles.Partition
      Float Selection Array=DataInterface() {propertyValues {FloatData [10,10,10,10,10,10,15,10,10,10,10,10,10,10,10]}}
     )
    )) {isSetParameters=true, module=SetVariables_#7}
   /SetVariables(Particles.VisibilityTag=&MakeCustomIntFromBool(
     Bool=&SetBoolByIntComparison(A=$Particles.PartitionParticles.Partition, B=5, Comparison Type=::NewEnumerator0)
     True INT=&MakeCustomIntFromBool(Bool=$User.bisHandHidden, True INT=1, False INT=0)
     False INT=0
    )) {isSetParameters=true}
  ]
  rendererClasses [&NiagaraSpriteRendererProperties]
  renderers [
   {rendererClass=&NiagaraSpriteRendererProperties}
   {rendererClass=&NiagaraSpriteRendererProperties, rendererIndex=1}
  ]
 }
}
summary {
 systemName=NS_PsychicAphid_Exo_Idle
 emitters [
  {emitterName=Glow_Center, rendererClasses [&NiagaraSpriteRendererProperties]}
  {emitterName=Glow_Shoulders, rendererClasses [&NiagaraSpriteRendererProperties]}
  {emitterName=BigGlows, rendererClasses [&NiagaraSpriteRendererProperties]}
  {emitterName=Glow_Head, rendererClasses [&NiagaraSpriteRendererProperties]}
 ]
 userVariables [
  User.bIsFrontEnd {type=NiagaraBool}
  User.bIsFrontEndPreview {type=NiagaraBool}
  User.bisHandHidden {type=NiagaraBool}
  User.bisHeadHidden {type=NiagaraBool}
 ]
}
