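# Physics asset for the PsychicAphid Exo male medium body cosmetic, set up for secondary motion on the left and right front lapels.
# Each lapel base bone carries a tiny capsule fully locked to spine_04, which has a large torso capsule.
# Each dynamic lapel capsule hinges off its base with limits that are wide on swing2 and tight on swing1 and twist, so it flaps mainly along one axis.
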PhysicsAsset /BRCosmetics/Characters/Player/Male/Medium/Bodies/M_MED_PsychicAphid_Exo/Meshes/M_MED_PsychicAphid_Exo_Physics
bodies {
 base_f_lape_l [{shapeType=Capsule, length=0.505, radius=0.505, shapeName=base_f_lape_l_capsule, rotation=<rot 0 -90 0>}]
 base_f_lape_r [{shapeType=Capsule, length=0.505, radius=0.505, shapeName=base_f_lape_r_capsule, rotation=<rot 0 -90 0>}]
 dyn_f_lape_l [{shapeType=Capsule, length=12.7298, radius=3, shapeName=dyn_f_lape_l_capsule, center=<6.6908 0 -1.5427>, rotation=<rot 90 0.0001 0.0003>}]
 dyn_f_lape_r [{
   shapeType=Capsule
   length=12.7298
   radius=3
   shapeName=dyn_f_lape_r_capsule
   center=<7.8689 0.0007 -1.4353>
   rotation=<rot 86.9874 -179.9951 -180>
  }]
 spine_04 [{
   shapeType=Capsule
   length=3.6373
   radius=12.0193
   shapeName=spine_04_capsule
   center=<-11.9238 -0.9358 -0.0049>
   rotation=<rot 0 179.6116 -1.5786>
  }]
}
constraints [
 {bone1Name=dyn_f_lape_l, bone2Name=base_f_lape_l, swing1LimitDegrees=5, swing2LimitDegrees=40, twistLimitDegrees=1}
 {
  bone1Name=base_f_lape_l
  bone2Name=spine_04
  swing1LimitDegrees=45
  swing1Motion=Locked
  swing2LimitDegrees=45
  swing2Motion=Locked
  twistLimitDegrees=45
  twistMotion=Locked
 }
 {
  bone1Name=base_f_lape_r
  bone2Name=spine_04
  swing1LimitDegrees=45
  swing1Motion=Locked
  swing2LimitDegrees=45
  swing2Motion=Locked
  twistLimitDegrees=45
  twistMotion=Locked
 }
 {bone1Name=dyn_f_lape_r, bone2Name=base_f_lape_r, swing1LimitDegrees=5, swing2LimitDegrees=40, twistLimitDegrees=1}
]
