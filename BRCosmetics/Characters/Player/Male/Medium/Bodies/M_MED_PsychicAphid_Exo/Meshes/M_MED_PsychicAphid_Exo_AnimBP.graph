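# Post-process style AnimBlueprint for the PsychicAphid Exo male body cosmetic, correcting armor and accessory bones atop the copied parent mesh pose.
# Weighted constraints place knee pads, teeth and leg armor, AnimDynamics swings front flaps, and the collar follows neck_01.
# Crouch-driven offsets then feed a Control Rig that receives downed (DBNO) and Bendy-emote alphas from the event graph.
# The gravity-override RigidBody and several duplicate nodes are disconnected from the output.
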AnimBlueprint /BRCosmetics/Characters/Player/Male/Medium/Bodies/M_MED_PsychicAphid_Exo/Meshes/M_MED_PsychicAphid_Exo_AnimBP
target_skeleton=&/M_MED_PsychicAphid_Exo_Skeleton
graphs [
 AnimGraph {nodes [
   n1 = Anim Dynamics - Bone: dyn_f_lape_l(Alpha=1, ExternalForce="0,0,0") {
    angularDampingOverride=1
    angularSpringConstant=300
    bAngularSpring=true
    bOverrideAngularDamping=true
    bUseGravityOverride=true
    simulationSpace=RootRelative
    alphaBoolBlend {blendOption=Linear}
    boundBone {boneName=dyn_f_lape_l}
    componentAppliedLinearAccClamp=<3000 3000 3000>
    componentLinearAccScale=<3 3 3>
    componentLinearVelScale=<3 3 3>
    gravityOverride {z=-980}
    physicsBodyDefinitions [{
      boundBone {boneName=dyn_f_lape_l}
      boxExtents=<7 7 7>
      constraintSetup {angularLimitsMax=<10 55 20>, angularLimitsMin=<-10 -1 -20>, angularTarget=<0 0 0>}
      localJointOffset=<0.5 -0.3 0>
     }]
   }
   Anim Dynamics - Bone: dyn_f_lape_r(ComponentPose=@n1, Alpha=1, ExternalForce="0,0,0") {
    angularDampingOverride=1
    angularSpringConstant=300
    bAngularSpring=true
    bOverrideAngularDamping=true
    bUseGravityOverride=true
    simulationSpace=RootRelative
    alphaBoolBlend {blendOption=Linear}
    boundBone {boneName=dyn_f_lape_r}
    componentAppliedLinearAccClamp=<3000 3000 3000>
    componentLinearAccScale=<3 3 3>
    componentLinearVelScale=<3 3 3>
    gravityOverride {z=-980}
    physicsBodyDefinitions [{
      boundBone {boneName=dyn_f_lape_r}
      boxExtents=<7 7 7>
      constraintSetup {angularLimitsMax=<10 55 20>, angularLimitsMin=<-10 -1 -20>, angularTarget=<0 0 0>}
      localJointOffset=<0.5 -0.3 0>
     }]
   }
   n2 = "Constraint - dyn_theeth_l_01 (PPPPPPPPP)"!(
    ComponentPose
    Alpha=1
    ConstraintWeights_0=0.0002
    ConstraintWeights_1:float=0.0001
    ConstraintWeights_2:float=0.1765
    ConstraintWeights_3:float=0.0205
    ConstraintWeights_4:float=0.4138
    ConstraintWeights_5:float=0.1331
    ConstraintWeights_6:float=0.0104
    ConstraintWeights_7:float=0.0105
    ConstraintWeights_8:float=0.2349
   ) -> (Pose) {
    alphaBoolBlend {blendOption=Linear}
    boneToModify {boneName=dyn_theeth_l_01}
    constraintSetup [
     {targetBone {boneName=spine_03}}
     {targetBone {boneName=spine_05}}
     {targetBone {boneName=clavicle_l}}
     {targetBone {boneName=upperarm_l}}
     {targetBone {boneName=upperarm_twist_01_l}}
     {targetBone {boneName=upperarm_twist_02_l}}
     {targetBone {boneName=deform_upperarm_bk_l}}
     {targetBone {boneName=deform_upperarm_fr_l}}
     {targetBone {boneName=deform_upperarm_up_l}}
    ]
    constraintWeights [1,1,1,1,1,1,1,1,1]
   }
   "Constraint - base_theeth_r (PPPPPPPP)"!(
    ComponentPose=@n2
    Alpha=1
    ConstraintWeights_0=0
    ConstraintWeights_1:float=0.0131
    ConstraintWeights_2:float=0.0106
    ConstraintWeights_3:float=0.9207
    ConstraintWeights_4:float=0.0408
    ConstraintWeights_5:float=0.0003
    ConstraintWeights_6:float=0.0018
    ConstraintWeights_7:float=0.0128
   ) -> (Pose) {
    alphaBoolBlend {blendOption=Linear}
    boneToModify {boneName=base_theeth_r}
    constraintSetup [
     {targetBone {boneName=spine_04}}
     {targetBone {boneName=clavicle_r}}
     {targetBone {boneName=upperarm_r}}
     {targetBone {boneName=upperarm_twist_01_r}}
     {targetBone {boneName=upperarm_twist_02_r}}
     {targetBone {boneName=deform_upperarm_bk_r}}
     {targetBone {boneName=deform_upperarm_fr_r}}
     {targetBone {boneName=deform_upperarm_up_r}}
    ]
    constraintWeights [1,1,1,1,1,1,1,1]
   }
   "Transform (Modify) Bone - Bone: base_collar_m"(Alpha=1, Translation="-0.2,0,0", Rotation="0,0,-18", Scale="1,1,1") {
    rotationMode=BMM_Additive
    translationMode=BMM_Additive
    alphaBoolBlend {blendOption=Linear}
    boneToModify {boneName=base_collar_m}
    rotation {roll=-18}
    translation {x=-0.2}
   }
   "Transform (Modify) Bone - Bone: dyn_collar_r"(Alpha=1, Translation="0,0,0", Rotation="-10,0,0", Scale="1,1,1") {
    rotationMode=BMM_Additive
    rotationSpace=BCS_BoneSpace
    translationMode=BMM_Additive
    alphaBoolBlend {blendOption=Linear}
    boneToModify {boneName=dyn_collar_r}
    rotation {pitch=-10}
   }
   "Transform (Modify) Bone - Bone: dyn_collar_l"(Alpha=1, Translation="0,0,0", Rotation="20,0,0", Scale="1,1,1") {
    rotationMode=BMM_Additive
    translationMode=BMM_Additive
    alphaBoolBlend {blendOption=Linear}
    boneToModify {boneName=dyn_collar_l}
    rotation {pitch=20}
   }
   n3 = "Transform (Modify) Bone - Bone: base_collar_m"(Alpha=1, Translation="-0.2,-1,0", Rotation="-15,0,-18", Scale="1,1,1") {
    rotationMode=BMM_Additive
    translationMode=BMM_Additive
    alphaBoolBlend {blendOption=Linear}
    boneToModify {boneName=base_collar_m}
    rotation {pitch=-15, roll=-18}
    translation=<-0.2 -1>
   }
   "Transform (Modify) Bone - Bone: dyn_collar_r"(ComponentPose=@n3, Alpha=1, Translation="0,0,0", Rotation="-10,0,0", Scale="1,1,1") {
    rotationMode=BMM_Additive
    rotationSpace=BCS_BoneSpace
    translationMode=BMM_Additive
    alphaBoolBlend {blendOption=Linear}
    boneToModify {boneName=dyn_collar_r}
    rotation {pitch=-10}
   }
   RigidBody!(ComponentPose, bAlphaBoolEnabled:bool, OverrideWorldGravity:Vector=Get Gravity Override(), ExternalForce) -> (Pose) {
    alphaInputType=Bool
    bOverrideWorldGravity=true
    overridePhysicsAsset=&/M_MED_PsychicAphid_Exo_Physics
    simulationSpace=BaseBoneSpace
    alphaBoolBlend {blendInTime=0.1, blendOption=Linear, blendOutTime=0.2}
    baseBoneRef {boneName=pelvis}
    simSpaceSettings {velocityScaleZ=0.3, worldAlpha=1}
   }
   Output Pose(Result=Use cached pose 'Collar Correct'()) {layerGroup=None}
   n4 = "Constraint - Knee_pad_l (PPPP)"!(
    ComponentPose=Local To Component(LocalPose=Copy Pose From Mesh() {bUseAttachedParent=true})
    Alpha=0.5
    ConstraintWeights_0=0.5
    ConstraintWeights_1:float=0
    ConstraintWeights_2:float=0.2
    ConstraintWeights_3:float=0.3
   ) -> (Pose) {
    alphaBoolBlend {blendOption=Linear}
    alphaScaleBiasClamp {bClampResult=true}
    boneToModify {boneName=Knee_pad_l}
    constraintSetup [
     {targetBone {boneName=thigh_l}}
     {targetBone {boneName=thigh_twist_01_l}}
     {targetBone {boneName=calf_l}}
     {targetBone {boneName=deform_knee_out_l}}
    ]
    constraintWeights [1,1,1,1]
   }
   n5 = "Constraint - Knee_pad_r (PPPPPPPP)"!(
    ComponentPose=@n4
    Alpha=0.5
    ConstraintWeights_0=0
    ConstraintWeights_1:float=0.5
    ConstraintWeights_2:float=0
    ConstraintWeights_3:float=0.2
    ConstraintWeights_4:float=0
    ConstraintWeights_5:float=0
    ConstraintWeights_6:float=0
    ConstraintWeights_7:float=0.3
   ) -> (Pose) {
    alphaBoolBlend {blendOption=Linear}
    alphaScaleBiasClamp {bClampResult=true}
    boneToModify {boneName=Knee_pad_r}
    constraintSetup [
     {targetBone {boneName=pelvis}}
     {targetBone {boneName=thigh_r}}
     {targetBone {boneName=calf_l}}
     {targetBone {boneName=calf_r}}
     {targetBone {boneName=calf_twist_02_r}}
     {targetBone {boneName=deform_groin_r}}
     {targetBone {boneName=deform_knee_out_r}}
     {targetBone {boneName=Knee_pad_r}}
    ]
    constraintWeights [1,1,1,1,1,1,1,1]
   }
   n6 = "Constraint - base_theeth_l (PPPPPPPPP)"!(
    ComponentPose=@n5
    Alpha=1
    ConstraintWeights_0=0.0006
    ConstraintWeights_1:float=0.0008
    ConstraintWeights_2:float=0.1135
    ConstraintWeights_3:float=0.0282
    ConstraintWeights_4:float=0.6119
    ConstraintWeights_5:float=0.1236
    ConstraintWeights_6:float=0.0134
    ConstraintWeights_7:float=0.0124
    ConstraintWeights_8:float=0.0956
   ) -> (Pose) {
    alphaBoolBlend {blendOption=Linear}
    boneToModify {boneName=base_theeth_l}
    constraintSetup [
     {targetBone {boneName=spine_03}}
     {targetBone {boneName=spine_05}}
     {targetBone {boneName=clavicle_l}}
     {targetBone {boneName=upperarm_l}}
     {targetBone {boneName=upperarm_twist_01_l}}
     {targetBone {boneName=upperarm_twist_02_l}}
     {targetBone {boneName=deform_upperarm_bk_l}}
     {targetBone {boneName=deform_upperarm_fr_l}}
     {targetBone {boneName=deform_upperarm_up_l}}
    ]
    constraintWeights [1,1,1,1,1,1,1,1,1]
   }
   n7 = "Constraint - base_theeth_r (PPPPPPPPPP)"!(
    ComponentPose=@n6
    Alpha=1
    ConstraintWeights_0=0.001
    ConstraintWeights_1:float=0.0002
    ConstraintWeights_2:float=0.1749
    ConstraintWeights_3:float=0.0716
    ConstraintWeights_4:float=0.4192
    ConstraintWeights_5:float=0.1809
    ConstraintWeights_6:float=0.0001
    ConstraintWeights_7:float=0.0261
    ConstraintWeights_8:float=0.0162
    ConstraintWeights_9:float=0.1097
   ) -> (Pose) {
    alphaBoolBlend {blendOption=Linear}
    boneToModify {boneName=base_theeth_r}
    constraintSetup [
     {targetBone {boneName=spine_03}}
     {targetBone {boneName=spine_05}}
     {targetBone {boneName=clavicle_r}}
     {targetBone {boneName=upperarm_r}}
     {targetBone {boneName=upperarm_twist_01_r}}
     {targetBone {boneName=upperarm_twist_02_r}}
     {targetBone {boneName=deform_elbow_in_r}}
     {targetBone {boneName=deform_upperarm_bk_r}}
     {targetBone {boneName=deform_upperarm_fr_r}}
     {targetBone {boneName=deform_upperarm_up_r}}
    ]
    constraintWeights [1,1,1,1,1,1,1,1,1,1]
   }
   n8 = "Constraint - leg_armor_cor_BS__r (PP)"!(ComponentPose=@n7, Alpha=1, ConstraintWeights_0=0, ConstraintWeights_1:float=1) -> (Pose) {
    alphaBoolBlend {blendOption=Linear}
    boneToModify {boneName=leg_armor_cor_BS__r}
    constraintSetup [{targetBone {boneName=thigh_r}}, {targetBone {boneName=thigh_twist_01_r}}]
    constraintWeights [1,1]
   }
   n9 = "Constraint - leg_armor_cor_BS_l (PPPPPPPP)"!(
    ComponentPose=@n8
    Alpha=1
    ConstraintWeights_0=0.0346
    ConstraintWeights_1:float=0
    ConstraintWeights_2:float=0
    ConstraintWeights_3:float=0
    ConstraintWeights_4:float=0.0311
    ConstraintWeights_5:float=0.8882
    ConstraintWeights_6:float=0.0186
    ConstraintWeights_7:float=0.0274
   ) -> (Pose) {
    alphaBoolBlend {blendOption=Linear}
    boneToModify {boneName=leg_armor_cor_BS_l}
    constraintSetup [
     {targetBone {boneName=pelvis}}
     {targetBone {boneName=spine_01}}
     {targetBone {boneName=spine_02}}
     {targetBone {boneName=spine_03}}
     {targetBone {boneName=thigh_l}}
     {targetBone {boneName=thigh_twist_01_l}}
     {targetBone {boneName=deform_glute_l}}
     {targetBone {boneName=deform_groin_l}}
    ]
    constraintWeights [1,1,1,1,1,1,1,1]
   }
   n10 = "Transform (Modify) Bone - Bone: leg_armor_cor_BS__r"(ComponentPose=@n9, Alpha=1, Translation="0,0,-0.3", Rotation="0.5,0,0", Scale="1,1,1") {
    rotationMode=BMM_Additive
    rotationSpace=BCS_BoneSpace
    translationMode=BMM_Additive
    translationSpace=BCS_BoneSpace
    alphaBoolBlend {blendOption=Linear}
    boneToModify {boneName=leg_armor_cor_BS__r}
    rotation {pitch=0.5}
    translation {z=-0.3}
   }
   n11 = "Transform (Modify) Bone - Bone: leg_armor_cor_BS_l"(ComponentPose=@n10, Alpha=1, Translation="0,0,0.3", Rotation="0,0,0", Scale="1,1,1") {
    translationMode=BMM_Additive
    translationSpace=BCS_BoneSpace
    alphaBoolBlend {blendOption=Linear}
    boneToModify {boneName=leg_armor_cor_BS_l}
    translation {z=0.3}
   }
   n12 = Anim Dynamics - Bone: dyn_f_lape_l(ComponentPose=@n11, Alpha=1, ExternalForce="0,0,0") {
    angularDampingOverride=0.999
    angularSpringConstant=300
    bAngularSpring=true
    bOverrideAngularDamping=true
    bUseGravityOverride=true
    simulationSpace=RootRelative
    alphaBoolBlend {blendOption=Linear}
    boundBone {boneName=dyn_f_lape_l}
    componentAppliedLinearAccClamp=<3000 3000 3000>
    componentLinearAccScale=<5 5 5>
    componentLinearVelScale=<5 5 5>
    gravityOverride {z=-980}
    physicsBodyDefinitions [{
      boundBone {boneName=dyn_f_lape_l}
      boxExtents=<7 7 7>
      constraintSetup {angularLimitsMax=<10 55 20>, angularLimitsMin=<-10 -1 -20>, angularTarget=<0 0 0>}
      localJointOffset=<0.7 0.5 0>
     }]
   }
   n13 = Anim Dynamics - Bone: dyn_f_lape_r(ComponentPose=@n12, Alpha=1, ExternalForce="0,0,0") {
    angularDampingOverride=0.999
    angularSpringConstant=300
    bAngularSpring=true
    bOverrideAngularDamping=true
    bUseGravityOverride=true
    simulationSpace=RootRelative
    alphaBoolBlend {blendOption=Linear}
    boundBone {boneName=dyn_f_lape_r}
    componentAppliedLinearAccClamp=<3000 3000 3000>
    componentLinearAccScale=<5 5 5>
    componentLinearVelScale=<5 5 5>
    gravityOverride {z=-980}
    physicsBodyDefinitions [{
      boundBone {boneName=dyn_f_lape_r}
      boxExtents=<7 7 7>
      constraintSetup {angularLimitsMax=<10 55 20>, angularLimitsMin=<-10 -1 -20>, angularTarget=<0 0 0>}
      localJointOffset=<0.7 -0.5 0>
     }]
   }
   Save cached pose 'Start Cashe'(Pose=Component To Local(ComponentPose=@n13)) {cacheName=Start Cashe}
   n14 = "Constraint - dyn_collar_m (P)"(ComponentPose=Local To Component(LocalPose=Use cached pose 'Start Cashe'()), Alpha=1, ConstraintWeights_0=1) {
    alphaBoolBlend {blendOption=Linear}
    boneToModify {boneName=dyn_collar_m}
    constraintSetup [{targetBone {boneName=neck_01}}]
    constraintWeights [1]
   }
   n15 = "Constraint - dyn_collar_r (P)"(ComponentPose=@n14, Alpha=1, ConstraintWeights_0=1) {
    alphaBoolBlend {blendOption=Linear}
    boneToModify {boneName=dyn_collar_r}
    constraintSetup [{targetBone {boneName=neck_01}}]
    constraintWeights [1]
   }
   n16 = "Constraint - dyn_collar_l (P)"(ComponentPose=@n15, Alpha=1, ConstraintWeights_0=1) {
    alphaBoolBlend {blendOption=Linear}
    boneToModify {boneName=dyn_collar_l}
    constraintSetup [{targetBone {boneName=neck_01}}]
    constraintWeights [1]
   }
   n17 = Get crouching alpha() -> (crouching alpha:double)
   n18 = "Transform (Modify) Bone - Bone: leg_armor_cor_BS__r"(ComponentPose=@n16, Alpha=@n17, Translation="0,0,-0.5", Rotation="0.5,0,0", Scale="1,1,1") {
    rotationMode=BMM_Additive
    rotationSpace=BCS_BoneSpace
    translationMode=BMM_Additive
    translationSpace=BCS_BoneSpace
    alphaBoolBlend {blendOption=Linear}
    boneToModify {boneName=leg_armor_cor_BS__r}
    rotation {pitch=0.5}
    translation {z=-0.5}
   }
   n19 = "Transform (Modify) Bone - Bone: leg_armor_cor_BS_l"(ComponentPose=@n18, Alpha=@n17, Translation="0,0,0.5", Rotation="0,0,0", Scale="1,1,1") {
    translationMode=BMM_Additive
    translationSpace=BCS_BoneSpace
    alphaBoolBlend {blendOption=Linear}
    boneToModify {boneName=leg_armor_cor_BS_l}
    translation {z=0.5}
   }
   n20 = "Transform (Modify) Bone - Bone: dyn_theeth_r_01"(ComponentPose=@n19, Alpha=@n17, Translation="0,0,-0.5", Rotation="0,0,0", Scale="1,1,1") {
    rotationMode=BMM_Additive
    rotationSpace=BCS_BoneSpace
    scaleMode=BMM_Additive
    translationMode=BMM_Additive
    alphaBoolBlend {blendOption=Linear}
    boneToModify {boneName=dyn_theeth_r_01}
    translation {z=-0.5}
   }
   n21 = Control Rig!(
    Source=Component To Local(ComponentPose=@n20)
    Alpha
    IsBenALPHA:double=Get IsBenALPA() -> (IsBenALPA:double)
    alphaDBNO:double=Get alphaDBNO() -> (alphaDBNO:double)
   ) -> (Pose) {
    alphaBoolBlend {blendOption=Linear}
    controlRigAssetReference {blueprintRigClass=&/M_MED_PsychicAphid_Exo_CtrlRig.M_MED_PsychicAphid_Exo_CtrlRig_C}
   }
   Save cached pose 'Collar Correct'(Pose=@n21) {cacheName=Collar Correct}
  ]}
 EventGraph {nodes [
   Event BlueprintUpdateAnimation()
   > n1 = Sequence()
   > n2 = Gravity Override State Function DEPRECATED(CharacterPartAnimInstance=Self-Reference(), GameTimeInSeconds=GetGameTimeInSeconds(), GravityOverrideParams=Get Gravity Override Params())
   > Set Gravity Override(Gravity Override=@n2)
   n3 = GetOwnerSkeletalMeshComponent()
   > n4 = Is Valid(InputObject=@n3)
   Set IsBenALPA(execute=@n4.Is Not Valid, IsBenALPA:double=0) -> (Output_Get:double)
   n5 = GetAnimInstance(self=@n3)
   Set IsBenALPA(
    execute=@n4.Is Valid
    IsBenALPA=SelectFloat(A=1, B=0, bPickA=Montage_IsPlaying(self=@n5, Montage=/BRCosmetics/Animation/Game/MainPlayer/Emotes/Bendy/Emote_Bendy_CMM_M))
   )
   Set alphaDBNO(execute=@n1.then_1, alphaDBNO:double=SelectFloat(A=1, B=0, bPickA=Get bIsDBNO())) -> (Output_Get:double)
   Event OnBeginCrouching()
   > Set crouching alpha(crouching alpha:double=1) -> (Output_Get:double)
   Event OnEndCrouching()
   > Set crouching alpha(crouching alpha=0)
  ]}
]
variables [
 alphaDBNO {type=number}
 crouching alpha {type=number}
 Gravity Override {type=Vector}
 Gravity Override Params {type=GravityOverrideParamsStruct}
 IsBenALPA {type=number}
]
