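# Control Rig for the PsychicAphid Exo character's face accessory mesh, set up to procedurally move its dynamic hair bones.
# When "Hair interp while fly" is true, it loops a hair bone array through Modify Transforms.
# Library functions build wobble quaternions from seeded random-range noise per Euler axis.
# The RandomScaledRotation call and "hair speed scale" are unconnected, so the loop applies only an identity transform.
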ControlRigBlueprint /BRCosmetics/Characters/Player/Male/Medium/Bodies/M_MED_PsychicAphid_Exo/Meshes/Parts/M_MED_PsychicAphid_Exo_FaceAcc_CtrlRig
graphs [
 New Function_1_ContainedGraph {nodes [
   RigVMFunctionEntryNode()
   Random = /Script/RigVM.RigVMFunction_RandomFloat(Duration=10, Maximum=15, Minimum=0, Seed=217) {is=struct RigVMUnitNode, template="Random::Execute(in Seed,in Minimum,in Maximum,in Duration,out Result)"}
   Multiply = &RigVMFunction_MathDoubleMul(B=-1, A=@Random.Result)
   Noise = /Script/RigVM.RigVMFunction_NoiseFloat(Frequency=0.5, Speed=2, Value=0.5, Minimum=@Multiply.Result, Maximum=@Random.Result) {is=struct RigVMUnitNode, template="Noise::Execute(in Value,in Speed,in Frequency,in Minimum,in Maximum,out Result)"}
   n1 = &RigVMFunction_RandomFloat(Duration=10, Maximum=20, Minimum=0, Seed=963) {template="Random::Execute(in Seed,in Minimum,in Maximum,in Duration,out Result)"}
   n2 = &RigVMFunction_RandomFloat(Duration=10, Maximum=15, Minimum=0, Seed=132) {template="Random::Execute(in Seed,in Minimum,in Maximum,in Duration,out Result)"}
   RigVMFunction_MathQuaternionFromEuler = &RigVMFunction_MathQuaternionFromEuler(
    Euler=(X=0,Y=0,Z=0)
    Euler.X=@Noise.Result
    Euler.Y=&RigVMFunction_NoiseFloat(Frequency=0.5, Speed=2, Value=0.3, Minimum=&RigVMFunction_MathDoubleMul(B=-1, A=@n1.Result).Result, Maximum=@n1.Result) {template="Noise::Execute(in Value,in Speed,in Frequency,in Minimum,in Maximum,out Result)"}.Result
    Euler.Z=&RigVMFunction_NoiseFloat(Frequency=0.5, Speed=2, Value=0.5, Minimum=&RigVMFunction_MathDoubleMul(B=-1, A=@n2.Result).Result, Maximum=@n2.Result) {template="Noise::Execute(in Value,in Speed,in Frequency,in Minimum,in Maximum,out Result)"}.Result
   )
   > RigVMFunctionReturnNode(Result=@RigVMFunction_MathQuaternionFromEuler.Result)
   RigVMFunction_MathFloatMake = /Script/RigVM.RigVMFunction_MathFloatMake(Value=0) {is=struct RigVMUnitNode}
   RigVMVariableNode() {variable=hair speed scale}
  ]}
 New Function_ContainedGraph {nodes [
   RigVMFunctionReturnNode(Argument_1=False, Result=RigVMFunctionEntryNode().Result)
   Random = &RigVMFunction_RandomFloat(Duration=10, Maximum=15, Minimum=0, Seed=217) {template="Random::Execute(in Seed,in Minimum,in Maximum,in Duration,out Result)"}
   Multiply = &RigVMFunction_MathDoubleMul(B=-1, A=@Random.Result)
   Noise = &RigVMFunction_NoiseFloat(Frequency=0.5, Speed=2, Value=0.5, Minimum=@Multiply.Result, Maximum=@Random.Result) {template="Noise::Execute(in Value,in Speed,in Frequency,in Minimum,in Maximum,out Result)"}
   n1 = &RigVMFunction_RandomFloat(Duration=10, Maximum=20, Minimum=0, Seed=963) {template="Random::Execute(in Seed,in Minimum,in Maximum,in Duration,out Result)"}
   n2 = &RigVMFunction_RandomFloat(Duration=10, Maximum=15, Minimum=0, Seed=132) {template="Random::Execute(in Seed,in Minimum,in Maximum,in Duration,out Result)"}
   RigVMFunction_MathQuaternionFromEuler = &RigVMFunction_MathQuaternionFromEuler(
    Euler=(X=0,Y=0,Z=0)
    Euler.X=@Noise.Result
    Euler.Y=&RigVMFunction_NoiseFloat(Frequency=0.5, Speed=2, Value=0.3, Minimum=&RigVMFunction_MathDoubleMul(B=-1, A=@n1.Result).Result, Maximum=@n1.Result) {template="Noise::Execute(in Value,in Speed,in Frequency,in Minimum,in Maximum,out Result)"}.Result
    Euler.Z=&RigVMFunction_NoiseFloat(Frequency=0.5, Speed=2, Value=0.5, Minimum=&RigVMFunction_MathDoubleMul(B=-1, A=@n2.Result).Result, Maximum=@n2.Result) {template="Noise::Execute(in Value,in Speed,in Frequency,in Minimum,in Maximum,out Result)"}.Result
   )
  ]}
 RigVMFunctionLibrary {nodes [
   New Function = RigVMCollapseNode(Result=())
   RandomScaledRotation = RigVMCollapseNode()
  ]}
 RigVMModel {nodes [
   RigUnit_BeginExecution = &RigUnit_BeginExecution()
   RigVMFunction_ControlFlowBranch = &RigVMFunction_ControlFlowBranch(ExecuteContext=@RigUnit_BeginExecution.ExecutePin, Condition=RigVMVariableNode() {variable=Hair interp while fly}.Value)
   RigUnit_ItemArray = &RigUnit_ItemArray(Items="((Type=Bone,Name=\"dyn_hair_dwn_fr_a\"),(Type=Bone,Name=\"dyn_hair_dwn_b\"),(Type=Bone,Name=\"dyn_hair_dwn_c\"),(Type=Bone,Name=\"dyn_hair_dwn_d\"),(Type=Bone,Name=\"dyn_hair_dwn_e\"),(Type=Bone,Name=\"dyn_hair_dwn_f\"),(Type=Bone,Name=\"dyn_hair_dwn_g\"),(Type=Bone,Nam ...(truncated)")
   DISPATCH_RigVMDispatch_ArrayGetNum = RigVMDispatchNode(Array=@RigUnit_ItemArray.Items) {template="DISPATCH_RigVMDispatch_ArrayGetNum(in Array,out Num)"}
   n1 = /Script/RigVM.RigVMFunction_ForLoopCount(Count=@DISPATCH_RigVMDispatch_ArrayGetNum.Num, ExecutePin=@RigVMFunction_ControlFlowBranch.True) {is=struct RigVMUnitNode}
   DISPATCH_RigVMDispatch_ArrayGetAtIndex = RigVMDispatchNode(Index=@n1.Index, Array=@RigUnit_ItemArray.Items) {template="DISPATCH_RigVMDispatch_ArrayGetAtIndex(in Array,in Index,out Element)"}
   > &RigUnit_ModifyTransforms(
    ItemToModify=((Item=(Type=Bone,Name="None"),Transform=(Rotation=(X=0,Y=0,Z=0,W=1),Translation=(X=0,Y=0,Z=0),Scale3D=(X=1,Y=1,Z=1))))
    Weight=1
    WeightMinimum=0
    ItemToModify.0.Item=@DISPATCH_RigVMDispatch_ArrayGetAtIndex.Element
   )
   RandomScaledRotation = RigVMFunctionReferenceNode()
   RigVMVariableNode() {variable=hair speed scale}
  ]}
]
hierarchy [
 root {loc=<0 0 0>, rot=<0 0 0 1>}
 pelvis {parent=root, loc=<0 -0.7071 93.1191>, rot=<0.0127 0.707 -0.0127 -0.707>}
 spine_01 {parent=pelvis, loc=<6.7494 -0.6388 0>, rot=<0 0 0.034 -0.9994>}
 spine_02 {parent=spine_01, loc=<7.2518 -0.035 0>, rot=<0 0 0.0238 -0.9997>}
 spine_03 {parent=spine_02, loc=<6.0104 0 0>, rot=<0 0 0.0149 -0.9999>}
 spine_04 {parent=spine_03, loc=<8.1024 0 0>, rot=<0 0 -0.0132 -0.9999>}
 spine_05 {parent=spine_04, loc=<10.0146 0 0>, rot=<0 0 -0.0015 -1>}
 neck_01 {parent=spine_05, loc=<11.3138 0.1529 0>, rot=<0 0 -0.2003 -0.9797>}
 neck_02 {parent=neck_01, loc=<4.5627 0 0>, rot=<0 0 0.0504 -0.9987>}
 head {parent=neck_02, loc=<5.3914 0 0>, rot=<0 0 0.115 -0.9934>}
 base_hair_dwn_fr_a {parent=head, loc=<10.8523 9.9447 1.7786>, rot=<0.4516 0.2316 0.8534 0.1184>}
 dyn_hair_dwn_fr_a {parent=base_hair_dwn_fr_a, loc=<0 0 0>, rot=<0 0 0 1>}
 base_hair_dwn_b {parent=head, loc=<9.7602 8.7349 4.3588>, rot=<0.4516 0.2316 0.8534 0.1184>}
 dyn_hair_dwn_b {parent=base_hair_dwn_b, loc=<0 0 0>, rot=<0 0 0 1>}
 base_hair_dwn_c {parent=head, loc=<11.3926 7.1689 5.5153>, rot=<0.4516 0.2316 0.8534 0.1184>}
 dyn_hair_dwn_c {parent=base_hair_dwn_c, loc=<0 0 0>, rot=<0 0 0 1>}
 base_hair_dwn_d {parent=head, loc=<8.9426 6.0051 6.1823>, rot=<0.1357 0.258 0.956 0.0329>}
 dyn_hair_dwn_d {parent=base_hair_dwn_d, loc=<0 0 0>, rot=<0 0 0 1>}
 base_hair_dwn_e {parent=head, loc=<10.3784 5.0992 6.8221>, rot=<0.1357 0.258 0.956 0.0329>}
 dyn_hair_dwn_e {parent=base_hair_dwn_e, loc=<0 0 0>, rot=<0 0 0 1>}
 base_hair_dwn_f {parent=head, loc=<9.136 3.4688 7.5544>, rot=<0.1514 0.1227 0.9808 -0.001>}
 dyn_hair_dwn_f {parent=base_hair_dwn_f, loc=<0 0 0>, rot=<0 0 0 1>}
 base_hair_dwn_g {parent=head, loc=<9.2252 1.2604 8.1016>, rot=<0.1514 0.1227 0.9808 -0.001>}
 dyn_hair_dwn_g {parent=base_hair_dwn_g, loc=<0 0 0>, rot=<0 0 0 1>}
 base_hair_dwn_h {parent=head, loc=<7.8886 -1.1964 7.9598>, rot=<0.0909 -0.023 0.9953 0.0229>}
 dyn_hair_dwn_h {parent=base_hair_dwn_h, loc=<0 0 0>, rot=<0 0 0 1>}
 base_hair_dwn_i {parent=head, loc=<7.7874 -3.2287 7.8751>, rot=<0.0936 -0.056 0.9902 0.0877>}
 dyn_hair_dwn_i {parent=base_hair_dwn_i, loc=<0 0 0>, rot=<0 0 0 1>}
 base_hair_dwn_bk_J {parent=head, loc=<7.7341 -5.1978 6.7714>, rot=<0.048 -0.3849 0.9175 0.0879>}
 dyn_hair_dwn_bk_J_01 {parent=base_hair_dwn_bk_J, loc=<0 0 0>, rot=<0 0 0 1>}
 dyn_hair_dwn_bk_J_02 {parent=dyn_hair_dwn_bk_J_01, loc=<3.8222 0 0>, rot=<0 0 0 1>}
 base_hair_dwn_side_S {parent=head, loc=<8.623 -4.6629 -5.8059>, rot=<-0.0692 -0.9736 0.1721 -0.1333>}
 dyn_hair_dwn_side_S_01 {parent=base_hair_dwn_side_S, loc=<0 0 0>, rot=<0 0 0 1>}
 base_hair_up_fr_a {parent=head, loc=<14.4764 9.097 2.0828>, rot=<0.7478 0.3873 0.5339 -0.0753>}
 dyn_hair_up_fr_a {parent=base_hair_up_fr_a, loc=<0 0 0>, rot=<0 0 0 1>}
 base_hair_up_m_b {parent=head, loc=<16.8331 1.5725 2.6829>, rot=<0.7303 0.1875 0.6324 0.1779>}
 dyn_hair_up_m_b {parent=base_hair_up_m_b, loc=<0 0 0>, rot=<0 0 0 1>}
 base_hair_up_m_c {parent=head, loc=<15.1913 -1.9455 0.2122>, rot=<0.7615 0.0132 0.6479 0.0113>}
 dyn_hair_up_m_c {parent=base_hair_up_m_c, loc=<0 0 0>, rot=<0 0 0 1>}
 Master_bk_hairsh {parent=head, loc=<7.5556 -7.0621 -0.221>, rot=<0.0447 -0.7153 0.6974 -0.0003>}
 base_hair_dwn_bk_L {parent=Master_bk_hairsh, loc=<3.7344 -5.2379 -2.437>, rot=<-0.3849 -0.0651 0.0658 -0.9183>}
 dyn_hair_dwn_bk_L {parent=base_hair_dwn_bk_L, loc=<0 0 0>, rot=<0 0 0 1>}
 base_hair_dwn_bk_k {parent=Master_bk_hairsh, loc=<0.6424 -5.1741 -0.269>, rot=<-0.3871 -0.0506 0.0313 -0.9201>}
 dyn_hair_dwn_bk_k_01 {parent=base_hair_dwn_bk_k, loc=<0 0 0>, rot=<0 0 0 1>}
 dyn_hair_dwn_bk_k_02 {parent=dyn_hair_dwn_bk_k_01, loc=<3.8222 0 0>, rot=<0 0 0 1>}
 base_hair_dwn_bk_m {parent=Master_bk_hairsh, loc=<0.7808 -2.4194 0.5179>, rot=<-0.0005 -0.0588 0.0077 -0.9982>}
 dyn_hair_dwn_bk_m_01 {parent=base_hair_dwn_bk_m, loc=<0 0 0>, rot=<0 0 0 1>}
 dyn_hair_dwn_bk_m_02 {parent=dyn_hair_dwn_bk_m_01, loc=<3.8222 0 0>, rot=<0 0 0 1>}
 base_hair_dwn_bk_n {parent=Master_bk_hairsh, loc=<4.5912 -3.122 -1.4981>, rot=<-0.3849 -0.0651 0.0658 -0.9183>}
 dyn_hair_dwn_bk_n {parent=base_hair_dwn_bk_n, loc=<0 0 0>, rot=<0 0 0 1>}
 base_hair_dwn_bk_O {parent=Master_bk_hairsh, loc=<4.805 -0.9414 -0.6171>, rot=<-0.1214 -0.0805 0.0456 -0.9883>}
 dyn_hair_dwn_bk_O {parent=base_hair_dwn_bk_O, loc=<0 0 0>, rot=<0 0 0 1>}
 base_hair_dwn_bk_p {parent=Master_bk_hairsh, loc=<0.9095 0.2893 1.2551>, rot=<0 0 0 1>}
 dyn_hair_dwn_bk_p_01 {parent=base_hair_dwn_bk_p, loc=<0 0 0>, rot=<0 0 0 1>}
 dyn_hair_dwn_bk_p_02 {parent=dyn_hair_dwn_bk_p_01, loc=<3.8222 0 0>, rot=<0 0 0 1>}
 base_hair_dwn_bk_Q {parent=Master_bk_hairsh, loc=<0.9245 1.9687 0.7363>, rot=<0.1742 -0.0102 -0.0574 -0.983>}
 dyn_hair_dwn_bk_Q_01 {parent=base_hair_dwn_bk_Q, loc=<0 0 0>, rot=<0 0 0 1>}
 dyn_hair_dwn_bk_Q_02 {parent=dyn_hair_dwn_bk_Q_01, loc=<3.8222 0 0>, rot=<0 0 0 1>}
 base_hair_dwn_side_R {parent=Master_bk_hairsh, loc=<0.6526 3.9421 -0.6165>, rot=<0.4205 0.0106 -0.1897 -0.8872>}
 dyn_hair_dwn_side_R_01 {parent=base_hair_dwn_side_R, loc=<0 0 0>, rot=<0 0 0 1>}
 dyn_hair_dwn_side_R_02 {parent=dyn_hair_dwn_side_R_01, loc=<3.8222 0 0>, rot=<0 0 0 1>}
]
variables [
 Hair interp while fly {type=bool, default=False}
 hair speed scale {type=double, default=0}
]
