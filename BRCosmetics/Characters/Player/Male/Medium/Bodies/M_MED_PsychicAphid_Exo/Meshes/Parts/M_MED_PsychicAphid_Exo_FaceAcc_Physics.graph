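# Physics asset for the PsychicAphid Exo male medium body's FaceAcc mesh, defining bodies and joints for hair physics.
# Capsule bodies cover the hair bones, plus a zero-length capsule on the head and an unconstrained collar capsule.
# Short base-to-dyn chains hang from the head or from Master_bk_hairsh, an angularly locked hub for most back strands.
# Tight swing and twist limits keep hair motion close to the groomed shape.
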PhysicsAsset /BRCosmetics/Characters/Player/Male/Medium/Bodies/M_MED_PsychicAphid_Exo/Meshes/Parts/M_MED_PsychicAphid_Exo_FaceAcc_Physics
bodies {
 Master_bk_hairsh [{shapeType=Capsule, length=0.505, radius=0.505, shapeName=Master_bk_hairsh_capsule, rotation=<rot 0 -90 0>}]
 base_collar_m [{
   shapeType=Capsule
   length=13.0801
   radius=2.3951
   shapeName=base_collar_m_capsule
   center=<0 0.0474 -0.5953>
   rotation=<rot -85.4459 90.0014 179.9986>
  }]
 base_hair_dwn_b [{shapeType=Capsule, length=0.505, radius=0.505, shapeName=base_hair_dwn_b_capsule, rotation=<rot 0 -90 0>}]
 base_hair_dwn_bk_J [{shapeType=Capsule, length=0.505, radius=0.505, shapeName=base_hair_dwn_bk_J_capsule, rotation=<rot 0 -90 0>}]
 base_hair_dwn_bk_L [{shapeType=Capsule, length=0.505, radius=0.505, shapeName=base_hair_dwn_bk_L_capsule, rotation=<rot 0 -90 0>}]
 base_hair_dwn_bk_O [{shapeType=Capsule, length=0.505, radius=0.505, shapeName=base_hair_dwn_bk_O_capsule, rotation=<rot 0 -90 0>}]
 base_hair_dwn_bk_Q [{shapeType=Capsule, length=1, radius=2, shapeName=base_hair_dwn_bk_Q_capsule, rotation=<rot 0 -90 0>}]
 base_hair_dwn_bk_k [{shapeType=Capsule, length=0.505, radius=0.505, shapeName=base_hair_dwn_bk_k_capsule, rotation=<rot 0 -90 0>}]
 base_hair_dwn_bk_m [{shapeType=Capsule, length=0.505, radius=0.505, shapeName=base_hair_dwn_bk_m_capsule, rotation=<rot 0 -90 0>}]
 base_hair_dwn_bk_n [{shapeType=Capsule, length=0.505, radius=0.505, shapeName=base_hair_dwn_bk_n_capsule, rotation=<rot 0 -90 0>}]
 base_hair_dwn_bk_p [{shapeType=Capsule, length=0.505, radius=0.505, shapeName=base_hair_dwn_bk_p_capsule, rotation=<rot 0 -90 0>}]
 base_hair_dwn_c [{shapeType=Capsule, length=0.505, radius=0.505, shapeName=base_hair_dwn_c_capsule, rotation=<rot 0 -90 0>}]
 base_hair_dwn_d [{shapeType=Capsule, length=0.505, radius=0.505, shapeName=base_hair_dwn_d_capsule, rotation=<rot 0 -90 0>}]
 base_hair_dwn_f [{shapeType=Capsule, length=0.505, radius=0.505, shapeName=base_hair_dwn_f_capsule, rotation=<rot 0 -90 0>}]
 base_hair_dwn_fr_a [{shapeType=Capsule, length=0.505, radius=0.505, shapeName=base_hair_dwn_fr_a_capsule, rotation=<rot 0 -90 0>}]
 base_hair_dwn_g [{shapeType=Capsule, length=0.505, radius=0.505, shapeName=base_hair_dwn_g_capsule, rotation=<rot 0 -90 0>}]
 base_hair_dwn_h [{shapeType=Capsule, length=0.505, radius=0.505, shapeName=base_hair_dwn_h_capsule, rotation=<rot 0 -90 0>}]
 base_hair_dwn_i [{shapeType=Capsule, length=0.505, radius=0.505, shapeName=base_hair_dwn_i_capsule, rotation=<rot 0 -90 0>}]
 base_hair_dwn_side_R [{shapeType=Capsule, length=0.505, radius=0.505, shapeName=base_hair_dwn_side_R_capsule, rotation=<rot 0 -90 0>}]
 base_hair_dwn_side_S [{shapeType=Capsule, length=0.505, radius=0.505, shapeName=base_hair_dwn_side_S_capsule, rotation=<rot 0 -90 0>}]
 base_hair_up_fr_a [{shapeType=Capsule, length=0.505, radius=0.505, shapeName=base_hair_up_fr_a_capsule, rotation=<rot 0 -90 0>}]
 base_hair_up_m_b [{shapeType=Capsule, length=0.505, radius=0.505, shapeName=base_hair_up_m_b_capsule, rotation=<rot 0 -90 0>}]
 base_hair_up_m_c [{shapeType=Capsule, length=0.505, radius=0.505, shapeName=base_hair_up_m_c_capsule, rotation=<rot 0 -90 0>}]
 dyn_hair_dwn_b [{shapeType=Capsule, length=0.505, radius=0.8961, shapeName=dyn_hair_dwn_b_capsule, rotation=<rot 0 -90 0>}]
 dyn_hair_dwn_bk_J_01 [{shapeType=Capsule, length=1, radius=2, shapeName=dyn_hair_dwn_bk_J_01_capsule, rotation=<rot 0 -90 0>}]
 dyn_hair_dwn_bk_J_02 [{
   shapeType=Capsule
   length=2.172
   radius=1.5701
   shapeName=dyn_hair_dwn_bk_J_02_capsule
   center=<2.8887 -0.1964 0.0742>
   rotation=<rot -0.0007 84.5098 -82.1316>
  }]
 dyn_hair_dwn_bk_L [{
   shapeType=Capsule
   length=2.172
   radius=1.5701
   shapeName=dyn_hair_dwn_bk_L_capsule
   center=<4.0546 0.2414 -0.2357>
   rotation=<rot 0 88.227 -76.8852>
  }]
 dyn_hair_dwn_bk_O [{shapeType=Capsule, length=0.505, radius=0.505, shapeName=dyn_hair_dwn_bk_O_capsule, rotation=<rot 0 -90 0>}]
 dyn_hair_dwn_bk_Q_01 [{shapeType=Capsule, length=1, radius=2, shapeName=dyn_hair_dwn_bk_Q_01_capsule, rotation=<rot 0 -90 0>}]
 dyn_hair_dwn_bk_Q_02 [{shapeType=Capsule, length=8.177, radius=1.5701, shapeName=dyn_hair_dwn_bk_Q_02_capsule, rotation=<rot 0 -90 101.1209>}]
 dyn_hair_dwn_bk_k_01 [{shapeType=Capsule, length=1, radius=2, shapeName=dyn_hair_dwn_bk_k_01_capsule, rotation=<rot 0 -90 0>}]
 dyn_hair_dwn_bk_k_02 [{shapeType=Capsule, length=4.7985, radius=1.5701, shapeName=dyn_hair_dwn_bk_k_02_capsule, rotation=<rot 0 -90.0001 101.1208>}]
 dyn_hair_dwn_bk_m_01 [{shapeType=Capsule, length=1, radius=2, shapeName=dyn_hair_dwn_bk_m_01_capsule, rotation=<rot 0 -90 0>}]
 dyn_hair_dwn_bk_m_02 [{
   shapeType=Capsule
   length=2.172
   radius=1.5701
   shapeName=dyn_hair_dwn_bk_m_02_capsule
   center=<3.6655 -0.6566 0.4588>
   rotation=<rot 0.001 -104.4474 -88.2889>
  }]
 dyn_hair_dwn_bk_n [{
   shapeType=Capsule
   length=2.172
   radius=1.5701
   shapeName=dyn_hair_dwn_bk_n_capsule
   center=<3.7624 -0.1005 0.6565>
   rotation=<rot -1.4554 -101.6707 -85.5751>
  }]
 dyn_hair_dwn_bk_p_01 [{shapeType=Capsule, length=1, radius=2, shapeName=dyn_hair_dwn_bk_p_01_capsule, rotation=<rot 0 -90 0>}]
 dyn_hair_dwn_bk_p_02 [{shapeType=Capsule, length=7.5979, radius=1.5701, shapeName=dyn_hair_dwn_bk_p_02_capsule, rotation=<rot 0 -90 101.1209>}]
 dyn_hair_dwn_c [{shapeType=Capsule, length=0.505, radius=0.8961, shapeName=dyn_hair_dwn_c_capsule, rotation=<rot 0 -90 0>}]
 dyn_hair_dwn_d [{shapeType=Capsule, length=0.505, radius=0.8961, shapeName=dyn_hair_dwn_d_capsule, rotation=<rot 0 -90 0>}]
 dyn_hair_dwn_f [{shapeType=Capsule, length=0.505, radius=0.8961, shapeName=dyn_hair_dwn_f_capsule, rotation=<rot 0 -90 0>}]
 dyn_hair_dwn_fr_a [{shapeType=Capsule, length=0.505, radius=0.8961, shapeName=dyn_hair_dwn_fr_a_capsule, rotation=<rot 0 -90 0>}]
 dyn_hair_dwn_g [{shapeType=Capsule, length=0.505, radius=0.8961, shapeName=dyn_hair_dwn_g_capsule, rotation=<rot 0 -90 0>}]
 dyn_hair_dwn_h [{shapeType=Capsule, length=0.505, radius=0.8961, shapeName=dyn_hair_dwn_h_capsule, rotation=<rot 0 -90 0>}]
 dyn_hair_dwn_i [{shapeType=Capsule, length=0.505, radius=0.8961, shapeName=dyn_hair_dwn_i_capsule, rotation=<rot 0 -90 0>}]
 dyn_hair_dwn_side_R_01 [{shapeType=Capsule, length=0.505, radius=0.8961, shapeName=dyn_hair_dwn_side_R_01_capsule, rotation=<rot 0 -90 0>}]
 dyn_hair_dwn_side_R_02 [{
   shapeType=Capsule
   length=9.3678
   radius=2.04
   shapeName=dyn_hair_dwn_side_R_02_capsule
   center=<0.4034 0.3905 -0.0793>
   rotation=<rot 0.2364 -91.2028 101.1184>
  }]
 dyn_hair_dwn_side_S_01 [{shapeType=Capsule, length=0.505, radius=0.8961, shapeName=dyn_hair_dwn_side_S_01_capsule, center=<0 0.114 0>}]
 dyn_hair_up_fr_a [{shapeType=Capsule, length=0.505, radius=0.8961, shapeName=dyn_hair_up_fr_a_capsule, rotation=<rot 0 -90 0>}]
 dyn_hair_up_m_b [{shapeType=Capsule, length=0.505, radius=0.8961, shapeName=dyn_hair_up_m_b_capsule, rotation=<rot 0 -90 0>}]
 dyn_hair_up_m_c [{shapeType=Capsule, length=3, radius=2, shapeName=dyn_hair_up_m_c_capsule, rotation=<rot 0 -84.5224 -105.1549>}]
 head [{shapeType=Capsule, length=0, radius=7.7222, center=<4.2942 1.542 0>}]
}
constraints [
 {bone1Name=base_hair_dwn_b, bone2Name=head, swing1LimitDegrees=10, swing2LimitDegrees=10, twistLimitDegrees=3}
 {bone1Name=dyn_hair_dwn_b, bone2Name=base_hair_dwn_b, swing1LimitDegrees=7, swing2LimitDegrees=12, twistLimitDegrees=3}
 {bone1Name=base_hair_dwn_bk_J, bone2Name=head, swing1LimitDegrees=10, swing2LimitDegrees=15, twistLimitDegrees=3}
 {bone1Name=dyn_hair_dwn_bk_J_01, bone2Name=base_hair_dwn_bk_J, swing1LimitDegrees=7, swing2LimitDegrees=12, twistLimitDegrees=3}
 {bone1Name=dyn_hair_dwn_bk_J_02, bone2Name=dyn_hair_dwn_bk_J_01, swing1LimitDegrees=7, swing2LimitDegrees=12, twistLimitDegrees=1}
 {bone1Name=base_hair_dwn_bk_L, bone2Name=Master_bk_hairsh, swing1LimitDegrees=10, swing2LimitDegrees=10, twistLimitDegrees=3}
 {bone1Name=dyn_hair_dwn_bk_L, bone2Name=base_hair_dwn_bk_L, swing1LimitDegrees=7, swing2LimitDegrees=12, twistLimitDegrees=1}
 {bone1Name=base_hair_dwn_bk_O, bone2Name=Master_bk_hairsh, swing1LimitDegrees=10, swing2LimitDegrees=20, twistLimitDegrees=3}
 {bone1Name=dyn_hair_dwn_bk_O, bone2Name=base_hair_dwn_bk_O, swing1LimitDegrees=7, swing2LimitDegrees=12, twistLimitDegrees=1}
 {bone1Name=base_hair_dwn_bk_Q, bone2Name=Master_bk_hairsh, swing1LimitDegrees=7, swing2LimitDegrees=20, twistLimitDegrees=3}
 {bone1Name=dyn_hair_dwn_bk_Q_01, bone2Name=base_hair_dwn_bk_Q, swing1LimitDegrees=7, swing2LimitDegrees=12, twistLimitDegrees=1}
 {bone1Name=dyn_hair_dwn_bk_Q_02, bone2Name=dyn_hair_dwn_bk_Q_01, swing1LimitDegrees=7, swing2LimitDegrees=12, twistLimitDegrees=1}
 {bone1Name=base_hair_dwn_bk_k, bone2Name=Master_bk_hairsh, swing1LimitDegrees=10, swing2LimitDegrees=10, twistLimitDegrees=3}
 {bone1Name=dyn_hair_dwn_bk_k_01, bone2Name=base_hair_dwn_bk_k, swing1LimitDegrees=7, swing2LimitDegrees=12, twistLimitDegrees=1}
 {bone1Name=dyn_hair_dwn_bk_k_02, bone2Name=dyn_hair_dwn_bk_k_01, swing1LimitDegrees=7, swing2LimitDegrees=12, twistLimitDegrees=1}
 {bone1Name=base_hair_dwn_bk_m, bone2Name=Master_bk_hairsh, swing1LimitDegrees=10, swing2LimitDegrees=10, twistLimitDegrees=3}
 {bone1Name=dyn_hair_dwn_bk_m_01, bone2Name=base_hair_dwn_bk_m, swing1LimitDegrees=7, swing2LimitDegrees=12, twistLimitDegrees=1}
 {bone1Name=dyn_hair_dwn_bk_m_02, bone2Name=dyn_hair_dwn_bk_m_01, swing1LimitDegrees=7, swing2LimitDegrees=12, twistLimitDegrees=1}
 {bone1Name=base_hair_dwn_bk_n, bone2Name=Master_bk_hairsh, swing1LimitDegrees=10, swing2LimitDegrees=10, twistLimitDegrees=1}
 {bone1Name=dyn_hair_dwn_bk_n, bone2Name=base_hair_dwn_bk_n, swing1LimitDegrees=7, swing2LimitDegrees=12, twistLimitDegrees=3}
 {bone1Name=base_hair_dwn_bk_p, bone2Name=Master_bk_hairsh, swing1LimitDegrees=10, swing2LimitDegrees=10, twistLimitDegrees=3}
 {bone1Name=dyn_hair_dwn_bk_p_01, bone2Name=base_hair_dwn_bk_p, swing1LimitDegrees=7, swing2LimitDegrees=12, twistLimitDegrees=1}
 {bone1Name=dyn_hair_dwn_bk_p_02, bone2Name=dyn_hair_dwn_bk_p_01, swing1LimitDegrees=7, swing2LimitDegrees=12, twistLimitDegrees=1}
 {bone1Name=base_hair_dwn_c, bone2Name=head, swing1LimitDegrees=10, swing2LimitDegrees=10, twistLimitDegrees=3}
 {bone1Name=dyn_hair_dwn_c, bone2Name=base_hair_dwn_c, swing1LimitDegrees=7, swing2LimitDegrees=12, twistLimitDegrees=3}
 {bone1Name=base_hair_dwn_d, bone2Name=head, swing1LimitDegrees=10, swing2LimitDegrees=10, twistLimitDegrees=3}
 {bone1Name=dyn_hair_dwn_d, bone2Name=base_hair_dwn_d, swing1LimitDegrees=7, swing2LimitDegrees=12, twistLimitDegrees=3}
 {bone1Name=base_hair_dwn_f, bone2Name=head, swing1LimitDegrees=10, swing2LimitDegrees=10, twistLimitDegrees=3}
 {bone1Name=dyn_hair_dwn_f, bone2Name=base_hair_dwn_f, swing1LimitDegrees=7, swing2LimitDegrees=12, twistLimitDegrees=3}
 {bone1Name=base_hair_dwn_fr_a, bone2Name=head, swing1LimitDegrees=10, swing2LimitDegrees=10, twistLimitDegrees=3}
 {bone1Name=dyn_hair_dwn_fr_a, bone2Name=base_hair_dwn_fr_a, swing1LimitDegrees=7, swing2LimitDegrees=12, twistLimitDegrees=3}
 {bone1Name=base_hair_dwn_g, bone2Name=head, swing1LimitDegrees=10, swing2LimitDegrees=10, twistLimitDegrees=3}
 {bone1Name=dyn_hair_dwn_g, bone2Name=base_hair_dwn_g, swing1LimitDegrees=7, swing2LimitDegrees=12, twistLimitDegrees=3}
 {bone1Name=base_hair_dwn_h, bone2Name=head, swing1LimitDegrees=10, swing2LimitDegrees=10, twistLimitDegrees=3}
 {bone1Name=dyn_hair_dwn_h, bone2Name=base_hair_dwn_h, swing1LimitDegrees=7, swing2LimitDegrees=12, twistLimitDegrees=3}
 {bone1Name=base_hair_dwn_i, bone2Name=head, swing1LimitDegrees=10, swing2LimitDegrees=10, twistLimitDegrees=3}
 {bone1Name=dyn_hair_dwn_i, bone2Name=base_hair_dwn_i, swing1LimitDegrees=7, swing2LimitDegrees=12, twistLimitDegrees=3}
 {bone1Name=base_hair_dwn_side_R, bone2Name=Master_bk_hairsh, swing1LimitDegrees=10, swing2LimitDegrees=20, twistLimitDegrees=3}
 {bone1Name=dyn_hair_dwn_side_R_01, bone2Name=base_hair_dwn_side_R, swing1LimitDegrees=7, swing2LimitDegrees=20, twistLimitDegrees=1}
 {bone1Name=dyn_hair_dwn_side_R_02, bone2Name=dyn_hair_dwn_side_R_01, swing1LimitDegrees=7, swing2LimitDegrees=20, twistLimitDegrees=3}
 {bone1Name=base_hair_dwn_side_S, bone2Name=head, swing1LimitDegrees=10, swing2LimitDegrees=10, twistLimitDegrees=3}
 {bone1Name=dyn_hair_dwn_side_S_01, bone2Name=base_hair_dwn_side_S, swing1LimitDegrees=7, swing2LimitDegrees=12, twistLimitDegrees=1}
 {bone1Name=base_hair_up_fr_a, bone2Name=head, swing1LimitDegrees=10, swing2LimitDegrees=10, twistLimitDegrees=3}
 {bone1Name=dyn_hair_up_fr_a, bone2Name=base_hair_up_fr_a, swing1LimitDegrees=7, swing2LimitDegrees=12, twistLimitDegrees=3}
 {bone1Name=base_hair_up_m_b, bone2Name=head, swing1LimitDegrees=10, swing2LimitDegrees=10, twistLimitDegrees=3}
 {bone1Name=dyn_hair_up_m_b, bone2Name=base_hair_up_m_b, swing1LimitDegrees=7, swing2LimitDegrees=12, twistLimitDegrees=3}
 {
  bone1Name=Master_bk_hairsh
  bone2Name=head
  swing1LimitDegrees=45
  swing1Motion=Locked
  swing2LimitDegrees=45
  swing2Motion=Locked
  twistLimitDegrees=45
  twistMotion=Locked
 }
 {bone1Name=base_hair_up_m_c, bone2Name=head, swing1LimitDegrees=7, swing2LimitDegrees=15, twistLimitDegrees=2}
 {bone1Name=dyn_hair_up_m_c, bone2Name=base_hair_up_m_c, swing1LimitDegrees=7, swing2LimitDegrees=12, twistLimitDegrees=2}
]
