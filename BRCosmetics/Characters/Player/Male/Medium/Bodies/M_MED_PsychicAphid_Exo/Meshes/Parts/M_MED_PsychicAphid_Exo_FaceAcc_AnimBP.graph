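# Animation Blueprint for the PsychicAphid Exo face-accessory part, adding secondary hair motion on top of the body's pose.
# The output path copies the body pose, then additively rotates back and side hair bones by a remapped up/down hair value.
# A head-space RigidBody node then simulates physics, weighted by hair speed with overridden gravity, and is cached for output.
# Separate AnimDynamics and Control Rig branches sit outside that output path.
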AnimBlueprint /BRCosmetics/Characters/Player/Male/Medium/Bodies/M_MED_PsychicAphid_Exo/Meshes/Parts/M_MED_PsychicAphid_Exo_FaceAcc_AnimBP
target_skeleton=&/M_MED_PsychicAphid_Exo_FaceAcc_Skeleton
graphs [
 AnimGraph {nodes [
   n1 = Anim Dynamics - Bone: dyn_hair_dwn_fr_a!(ComponentPose, Alpha=1, GravityOverride:Vector="0,0,-980", ExternalForce="0,0,0") -> (Pose) {
    angularDampingOverride=0.1
    angularSpringConstant=10
    bAngularSpring=true
    bOverrideAngularDamping=true
    bUseGravityOverride=true
    simulationSpace=RootRelative
    alphaBoolBlend {blendOption=Linear}
    alphaScaleBiasClamp {bClampResult=true}
    boundBone {boneName=dyn_hair_dwn_fr_a}
    componentAppliedLinearAccClamp=<3000 3000 3000>
    componentLinearAccScale=<0.1 0.2>
    componentLinearVelScale=<0.1 0.2 0.1>
    gravityOverride {z=-980}
    physicsBodyDefinitions [{
      boundBone {boneName=dyn_hair_dwn_fr_a}
      boxExtents=<10 10 10>
      constraintSetup {angularLimitsMax=<7 10 15>, angularLimitsMin=<-7 -33 -15>, angularTarget=<1 0 0>}
      localJointOffset=<0.6 0 0>
     }]
    relativeSpaceBone {boneName=spine_05}
   }
   n2 = Anim Dynamics - Bone: dyn_hair_dwn_b!(ComponentPose=@n1, Alpha=1, GravityOverride:Vector="0,0,-980", ExternalForce="0,0,0") -> (Pose) {
    angularDampingOverride=0.1
    angularSpringConstant=10
    bAngularSpring=true
    bOverrideAngularDamping=true
    bUseGravityOverride=true
    simulationSpace=RootRelative
    alphaBoolBlend {blendOption=Linear}
    alphaScaleBiasClamp {bClampResult=true}
    boundBone {boneName=dyn_hair_dwn_b}
    componentAppliedLinearAccClamp=<3000 3000 3000>
    componentLinearAccScale=<0.1 0.2>
    componentLinearVelScale=<0.1 0.2 0.1>
    gravityOverride {z=-980}
    physicsBodyDefinitions [{
      boundBone {boneName=dyn_hair_dwn_b}
      boxExtents=<10 10 10>
      constraintSetup {angularLimitsMax=<7 10 15>, angularLimitsMin=<-7 -33 -15>, angularTarget=<1 0 0>}
      localJointOffset=<0.6 0 0>
     }]
    relativeSpaceBone {boneName=spine_05}
   }
   n3 = Anim Dynamics - Bone: dyn_hair_dwn_c!(ComponentPose=@n2, Alpha=1, GravityOverride:Vector="0,0,-980", ExternalForce="0,0,0") -> (Pose) {
    angularDampingOverride=0.1
    angularSpringConstant=10
    bAngularSpring=true
    bOverrideAngularDamping=true
    bUseGravityOverride=true
    simulationSpace=RootRelative
    alphaBoolBlend {blendOption=Linear}
    alphaScaleBiasClamp {bClampResult=true}
    boundBone {boneName=dyn_hair_dwn_c}
    componentAppliedLinearAccClamp=<3000 3000 3000>
    componentLinearAccScale=<0.1 0.2>
    componentLinearVelScale=<0.1 0.2 0.1>
    gravityOverride {z=-980}
    physicsBodyDefinitions [{
      boundBone {boneName=dyn_hair_dwn_c}
      boxExtents=<10 10 10>
      constraintSetup {angularLimitsMax=<7 10 15>, angularLimitsMin=<-7 -33 -15>, angularTarget=<1 0 0>}
      localJointOffset=<0.6 0 0>
     }]
    relativeSpaceBone {boneName=spine_05}
   }
   n4 = Anim Dynamics - Bone: dyn_hair_dwn_d!(ComponentPose=@n3, Alpha=1, GravityOverride:Vector="0,0,-980", ExternalForce="0,0,0") -> (Pose) {
    angularDampingOverride=0.1
    angularSpringConstant=10
    bAngularSpring=true
    bOverrideAngularDamping=true
    bUseGravityOverride=true
    simulationSpace=RootRelative
    alphaBoolBlend {blendOption=Linear}
    alphaScaleBiasClamp {bClampResult=true}
    boundBone {boneName=dyn_hair_dwn_d}
    componentAppliedLinearAccClamp=<3000 3000 3000>
    componentLinearAccScale=<0.1 0.2>
    componentLinearVelScale=<0.1 0.2 0.1>
    gravityOverride {z=-980}
    physicsBodyDefinitions [{
      boundBone {boneName=dyn_hair_dwn_d}
      boxExtents=<10 10 10>
      constraintSetup {angularLimitsMax=<7 10 15>, angularLimitsMin=<-7 -33 -15>, angularTarget=<1 0 0>}
      localJointOffset=<0.6 0 0>
     }]
    relativeSpaceBone {boneName=spine_05}
   }
   n5 = Anim Dynamics - Bone: dyn_hair_dwn_e!(ComponentPose=@n4, Alpha=1, GravityOverride:Vector="0,0,-980", ExternalForce="0,0,0") -> (Pose) {
    angularDampingOverride=0.1
    angularSpringConstant=10
    bAngularSpring=true
    bOverrideAngularDamping=true
    bUseGravityOverride=true
    simulationSpace=RootRelative
    alphaBoolBlend {blendOption=Linear}
    alphaScaleBiasClamp {bClampResult=true}
    boundBone {boneName=dyn_hair_dwn_e}
    componentAppliedLinearAccClamp=<3000 3000 3000>
    componentLinearAccScale=<0.1 0.2>
    componentLinearVelScale=<0.1 0.2 0.1>
    gravityOverride {z=-980}
    physicsBodyDefinitions [{
      boundBone {boneName=dyn_hair_dwn_e}
      boxExtents=<10 10 10>
      constraintSetup {angularLimitsMax=<7 10 15>, angularLimitsMin=<-7 -33 -15>, angularTarget=<1 0 0>}
      localJointOffset=<0.6 0 0>
     }]
    relativeSpaceBone {boneName=spine_05}
   }
   n6 = Anim Dynamics - Bone: dyn_hair_dwn_f!(ComponentPose=@n5, Alpha=1, GravityOverride:Vector="0,0,-980", ExternalForce="0,0,0") -> (Pose) {
    angularDampingOverride=0.1
    angularSpringConstant=10
    bAngularSpring=true
    bOverrideAngularDamping=true
    bUseGravityOverride=true
    simulationSpace=RootRelative
    alphaBoolBlend {blendOption=Linear}
    alphaScaleBiasClamp {bClampResult=true}
    boundBone {boneName=dyn_hair_dwn_f}
    componentAppliedLinearAccClamp=<3000 3000 3000>
    componentLinearAccScale=<0.1 0.2>
    componentLinearVelScale=<0.1 0.2 0.1>
    gravityOverride {z=-980}
    physicsBodyDefinitions [{
      boundBone {boneName=dyn_hair_dwn_f}
      boxExtents=<10 10 10>
      constraintSetup {angularLimitsMax=<7 10 15>, angularLimitsMin=<-7 -33 -15>, angularTarget=<1 0 0>}
      localJointOffset=<0.6 0 0>
     }]
    relativeSpaceBone {boneName=spine_05}
   }
   n7 = Anim Dynamics - Bone: dyn_hair_dwn_g!(ComponentPose=@n6, Alpha=1, GravityOverride:Vector="0,0,-980", ExternalForce="0,0,0") -> (Pose) {
    angularDampingOverride=0.1
    angularSpringConstant=10
    bAngularSpring=true
    bOverrideAngularDamping=true
    bUseGravityOverride=true
    simulationSpace=RootRelative
    alphaBoolBlend {blendOption=Linear}
    alphaScaleBiasClamp {bClampResult=true}
    boundBone {boneName=dyn_hair_dwn_g}
    componentAppliedLinearAccClamp=<3000 3000 3000>
    componentLinearAccScale=<0.1 0.2>
    componentLinearVelScale=<0.1 0.2 0.1>
    gravityOverride {z=-980}
    physicsBodyDefinitions [{
      boundBone {boneName=dyn_hair_dwn_g}
      boxExtents=<10 10 10>
      constraintSetup {angularLimitsMax=<7 10 15>, angularLimitsMin=<-7 -33 -15>, angularTarget=<1 0 0>}
      localJointOffset=<0.6 0 0>
     }]
    relativeSpaceBone {boneName=spine_05}
   }
   n8 = Anim Dynamics - Bone: dyn_hair_dwn_h!(ComponentPose=@n7, Alpha=1, GravityOverride:Vector="0,0,-980", ExternalForce="0,0,0") -> (Pose) {
    angularDampingOverride=0.1
    angularSpringConstant=10
    bAngularSpring=true
    bOverrideAngularDamping=true
    bUseGravityOverride=true
    simulationSpace=RootRelative
    alphaBoolBlend {blendOption=Linear}
    alphaScaleBiasClamp {bClampResult=true}
    boundBone {boneName=dyn_hair_dwn_h}
    componentAppliedLinearAccClamp=<3000 3000 3000>
    componentLinearAccScale=<0.1 0.2>
    componentLinearVelScale=<0.1 0.2 0.1>
    gravityOverride {z=-980}
    physicsBodyDefinitions [{
      boundBone {boneName=dyn_hair_dwn_h}
      boxExtents=<10 10 10>
      constraintSetup {angularLimitsMax=<7 10 15>, angularLimitsMin=<-7 -33 -15>, angularTarget=<1 0 0>}
      localJointOffset=<0.6 0 0>
     }]
    relativeSpaceBone {boneName=spine_05}
   }
   n9 = Anim Dynamics - Bone: dyn_hair_dwn_i!(ComponentPose=@n8, Alpha=1, GravityOverride:Vector="0,0,-980", ExternalForce="0,0,0") -> (Pose) {
    angularDampingOverride=0.1
    angularSpringConstant=10
    bAngularSpring=true
    bOverrideAngularDamping=true
    bUseGravityOverride=true
    simulationSpace=RootRelative
    alphaBoolBlend {blendOption=Linear}
    alphaScaleBiasClamp {bClampResult=true}
    boundBone {boneName=dyn_hair_dwn_i}
    componentAppliedLinearAccClamp=<3000 3000 3000>
    componentLinearAccScale=<0.1 0.2>
    componentLinearVelScale=<0.1 0.2 0.1>
    gravityOverride {z=-980}
    physicsBodyDefinitions [{
      boundBone {boneName=dyn_hair_dwn_i}
      boxExtents=<10 10 10>
      constraintSetup {angularLimitsMax=<7 10 15>, angularLimitsMin=<-7 -33 -15>, angularTarget=<1 0 0>}
      localJointOffset=<0.6 0 0>
     }]
    relativeSpaceBone {boneName=spine_05}
   }
   n10 = Anim Dynamics - Bone: dyn_hair_dwn_bk_J_02!(ComponentPose=@n9, Alpha=1, GravityOverride:Vector="0,0,-980", ExternalForce="0,0,0") -> (Pose) {
    angularDampingOverride=0.1
    angularSpringConstant=10
    bAngularSpring=true
    bOverrideAngularDamping=true
    bUseGravityOverride=true
    simulationSpace=RootRelative
    alphaBoolBlend {blendOption=Linear}
    alphaScaleBiasClamp {bClampResult=true}
    boundBone {boneName=dyn_hair_dwn_bk_J_02}
    componentAppliedLinearAccClamp=<3000 3000 3000>
    componentLinearAccScale=<0.1 0.2>
    componentLinearVelScale=<0.1 0.2 0.1>
    gravityOverride {z=-980}
    physicsBodyDefinitions [{
      boundBone {boneName=dyn_hair_dwn_bk_J_02}
      boxExtents=<10 10 10>
      constraintSetup {angularLimitsMax=<7 10 15>, angularLimitsMin=<-7 -33 -15>, angularTarget=<1 0 0>}
      localJointOffset=<0.6 0 0>
     }]
    relativeSpaceBone {boneName=spine_05}
   }
   n11 = Anim Dynamics - Bone: dyn_hair_dwn_side_S_01!(ComponentPose=@n10, Alpha=1, GravityOverride:Vector="0,0,-980", ExternalForce="0,0,0") -> (Pose) {
    angularDampingOverride=0.1
    angularSpringConstant=10
    bAngularSpring=true
    bOverrideAngularDamping=true
    bUseGravityOverride=true
    simulationSpace=RootRelative
    alphaBoolBlend {blendOption=Linear}
    alphaScaleBiasClamp {bClampResult=true}
    boundBone {boneName=dyn_hair_dwn_side_S_01}
    componentAppliedLinearAccClamp=<3000 3000 3000>
    componentLinearAccScale=<0.1 0.2>
    componentLinearVelScale=<0.1 0.2 0.1>
    gravityOverride {z=-980}
    physicsBodyDefinitions [{
      boundBone {boneName=dyn_hair_dwn_side_S_01}
      boxExtents=<10 10 10>
      constraintSetup {angularLimitsMax=<7 10 15>, angularLimitsMin=<-7 -33 -15>, angularTarget=<1 0 0>}
      localJointOffset=<0.6 0 0>
     }]
    relativeSpaceBone {boneName=spine_05}
   }
   n12 = Anim Dynamics - Bone: dyn_hair_dwn_side_S_01(ComponentPose=@n11, Alpha=1, GravityOverride="0,0,-980", ExternalForce="0,0,0") {
    angularDampingOverride=0.1
    angularSpringConstant=10
    bAngularSpring=true
    bOverrideAngularDamping=true
    bUseGravityOverride=true
    simulationSpace=RootRelative
    alphaBoolBlend {blendOption=Linear}
    alphaScaleBiasClamp {bClampResult=true}
    boundBone {boneName=dyn_hair_dwn_side_S_01}
    componentAppliedLinearAccClamp=<3000 3000 3000>
    componentLinearAccScale=<0.1 0.2>
    componentLinearVelScale=<0.1 0.2 0.1>
    gravityOverride {z=-980}
    physicsBodyDefinitions [{
      boundBone {boneName=dyn_hair_dwn_side_S_01}
      boxExtents=<10 10 10>
      constraintSetup {angularLimitsMax=<7 10 15>, angularLimitsMin=<-7 -33 -15>, angularTarget=<1 0 0>}
      localJointOffset=<0.6 0 0>
     }]
    relativeSpaceBone {boneName=spine_05}
   }
   n13 = Anim Dynamics - Bone: dyn_hair_up_fr_a!(ComponentPose=@n12, Alpha=1, GravityOverride:Vector="0,0,-980", ExternalForce="0,0,0") -> (Pose) {
    angularDampingOverride=0.1
    angularSpringConstant=10
    bAngularSpring=true
    bOverrideAngularDamping=true
    bUseGravityOverride=true
    simulationSpace=RootRelative
    alphaBoolBlend {blendOption=Linear}
    alphaScaleBiasClamp {bClampResult=true}
    boundBone {boneName=dyn_hair_up_fr_a}
    componentAppliedLinearAccClamp=<3000 3000 3000>
    componentLinearAccScale=<0.1 0.2>
    componentLinearVelScale=<0.1 0.2 0.1>
    gravityOverride {z=-980}
    physicsBodyDefinitions [{
      boundBone {boneName=dyn_hair_up_fr_a}
      boxExtents=<10 10 10>
      constraintSetup {angularLimitsMax=<3 10 15>, angularLimitsMin=<-3 -33 -15>, angularTarget=<1 0 0>}
      localJointOffset=<0.6 0 0>
     }]
    relativeSpaceBone {boneName=spine_05}
   }
   n14 = Anim Dynamics - Bone: dyn_hair_up_m_b!(ComponentPose=@n13, Alpha=1, GravityOverride:Vector="0,0,-980", ExternalForce="0,0,0") -> (Pose) {
    angularDampingOverride=0.1
    angularSpringConstant=10
    bAngularSpring=true
    bOverrideAngularDamping=true
    bUseGravityOverride=true
    simulationSpace=RootRelative
    alphaBoolBlend {blendOption=Linear}
    alphaScaleBiasClamp {bClampResult=true}
    boundBone {boneName=dyn_hair_up_m_b}
    componentAppliedLinearAccClamp=<3000 3000 3000>
    componentLinearAccScale=<0.1 0.2>
    componentLinearVelScale=<0.1 0.2 0.1>
    gravityOverride {z=-980}
    physicsBodyDefinitions [{
      boundBone {boneName=dyn_hair_up_m_b}
      boxExtents=<10 10 10>
      constraintSetup {angularLimitsMax=<3 10 15>, angularLimitsMin=<-3 -33 -15>, angularTarget=<1 0 0>}
      localJointOffset=<0.6 0 0>
     }]
    relativeSpaceBone {boneName=spine_05}
   }
   n15 = Anim Dynamics - Bone: dyn_hair_up_m_c!(ComponentPose=@n14, Alpha=1, GravityOverride:Vector="0,0,-980", ExternalForce="0,0,0") -> (Pose) {
    angularDampingOverride=0.1
    angularSpringConstant=10
    bAngularSpring=true
    bOverrideAngularDamping=true
    bUseGravityOverride=true
    simulationSpace=RootRelative
    alphaBoolBlend {blendOption=Linear}
    alphaScaleBiasClamp {bClampResult=true}
    boundBone {boneName=dyn_hair_up_m_c}
    componentAppliedLinearAccClamp=<3000 3000 3000>
    componentLinearAccScale=<0.1 0.2>
    componentLinearVelScale=<0.1 0.2 0.1>
    gravityOverride {z=-980}
    physicsBodyDefinitions [{
      boundBone {boneName=dyn_hair_up_m_c}
      boxExtents=<10 10 10>
      constraintSetup {angularLimitsMax=<3 10 15>, angularLimitsMin=<-3 -33 -15>, angularTarget=<1 0 0>}
      localJointOffset=<0.6 0 0>
     }]
    relativeSpaceBone {boneName=spine_05}
   }
   n16 = Anim Dynamics - Bone: dyn_hair_dwn_bk_L!(ComponentPose=@n15, Alpha=1, GravityOverride:Vector="0,0,-980", ExternalForce="0,0,0") -> (Pose) {
    angularDampingOverride=0.1
    angularSpringConstant=10
    bAngularSpring=true
    bOverrideAngularDamping=true
    bUseGravityOverride=true
    simulationSpace=RootRelative
    alphaBoolBlend {blendOption=Linear}
    alphaScaleBiasClamp {bClampResult=true}
    boundBone {boneName=dyn_hair_dwn_bk_L}
    componentAppliedLinearAccClamp=<3000 3000 3000>
    componentLinearAccScale=<0.1 0.2>
    componentLinearVelScale=<0.1 0.2 0.1>
    gravityOverride {z=-980}
    physicsBodyDefinitions [{
      boundBone {boneName=dyn_hair_dwn_bk_L}
      boxExtents=<10 10 10>
      constraintSetup {angularLimitsMax=<3 10 15>, angularLimitsMin=<-3 -33 -15>, angularTarget=<1 0 0>}
      localJointOffset=<0.6 0 0>
     }]
    relativeSpaceBone {boneName=spine_05}
   }
   n17 = Anim Dynamics - Bone: dyn_hair_dwn_bk_k_02!(ComponentPose=@n16, Alpha=1, GravityOverride:Vector="0,0,-980", ExternalForce="0,0,0") -> (Pose) {
    angularDampingOverride=0.1
    angularSpringConstant=10
    bAngularSpring=true
    bOverrideAngularDamping=true
    bUseGravityOverride=true
    simulationSpace=RootRelative
    alphaBoolBlend {blendOption=Linear}
    alphaScaleBiasClamp {bClampResult=true}
    boundBone {boneName=dyn_hair_dwn_bk_k_02}
    componentAppliedLinearAccClamp=<3000 3000 3000>
    componentLinearAccScale=<0.1 0.2>
    componentLinearVelScale=<0.1 0.2 0.1>
    gravityOverride {z=-980}
    physicsBodyDefinitions [{
      boundBone {boneName=dyn_hair_dwn_bk_k_02}
      boxExtents=<10 10 10>
      constraintSetup {angularLimitsMax=<3 10 15>, angularLimitsMin=<-3 -33 -15>, angularTarget=<1 0 0>}
      localJointOffset=<0.6 0 0>
     }]
    relativeSpaceBone {boneName=spine_05}
   }
   n18 = Anim Dynamics - Bone: dyn_hair_dwn_bk_m_02!(ComponentPose=@n17, Alpha=1, GravityOverride:Vector="0,0,-980", ExternalForce="0,0,0") -> (Pose) {
    angularDampingOverride=0.1
    angularSpringConstant=10
    bAngularSpring=true
    bOverrideAngularDamping=true
    bUseGravityOverride=true
    simulationSpace=RootRelative
    alphaBoolBlend {blendOption=Linear}
    alphaScaleBiasClamp {bClampResult=true}
    boundBone {boneName=dyn_hair_dwn_bk_m_02}
    componentAppliedLinearAccClamp=<3000 3000 3000>
    componentLinearAccScale=<0.1 0.2>
    componentLinearVelScale=<0.1 0.2 0.1>
    gravityOverride {z=-980}
    physicsBodyDefinitions [{
      boundBone {boneName=dyn_hair_dwn_bk_m_02}
      boxExtents=<10 10 10>
      constraintSetup {angularLimitsMax=<3 10 15>, angularLimitsMin=<-3 -33 -15>, angularTarget=<1 0 0>}
      localJointOffset=<0.6 0 0>
     }]
    relativeSpaceBone {boneName=spine_05}
   }
   n19 = Anim Dynamics - Bone: dyn_hair_dwn_bk_n!(ComponentPose=@n18, Alpha=1, GravityOverride:Vector="0,0,-980", ExternalForce="0,0,0") -> (Pose) {
    angularDampingOverride=0.1
    angularSpringConstant=10
    bAngularSpring=true
    bOverrideAngularDamping=true
    bUseGravityOverride=true
    simulationSpace=RootRelative
    alphaBoolBlend {blendOption=Linear}
    alphaScaleBiasClamp {bClampResult=true}
    boundBone {boneName=dyn_hair_dwn_bk_n}
    componentAppliedLinearAccClamp=<3000 3000 3000>
    componentLinearAccScale=<0.1 0.2>
    componentLinearVelScale=<0.1 0.2 0.1>
    gravityOverride {z=-980}
    physicsBodyDefinitions [{
      boundBone {boneName=dyn_hair_dwn_bk_n}
      boxExtents=<10 10 10>
      constraintSetup {angularLimitsMax=<3 10 15>, angularLimitsMin=<-3 -33 -15>, angularTarget=<1 0 0>}
      localJointOffset=<0.6 0 0>
     }]
    relativeSpaceBone {boneName=spine_05}
   }
   n20 = Anim Dynamics - Bone: dyn_hair_dwn_bk_O!(ComponentPose=@n19, Alpha=1, GravityOverride:Vector="0,0,-980", ExternalForce="0,0,0") -> (Pose) {
    angularDampingOverride=0.1
    angularSpringConstant=10
    bAngularSpring=true
    bOverrideAngularDamping=true
    bUseGravityOverride=true
    simulationSpace=RootRelative
    alphaBoolBlend {blendOption=Linear}
    alphaScaleBiasClamp {bClampResult=true}
    boundBone {boneName=dyn_hair_dwn_bk_O}
    componentAppliedLinearAccClamp=<3000 3000 3000>
    componentLinearAccScale=<0.1 0.2>
    componentLinearVelScale=<0.1 0.2 0.1>
    gravityOverride {z=-980}
    physicsBodyDefinitions [{
      boundBone {boneName=dyn_hair_dwn_bk_O}
      boxExtents=<10 10 10>
      constraintSetup {angularLimitsMax=<3 10 15>, angularLimitsMin=<-3 -33 -15>, angularTarget=<1 0 0>}
      localJointOffset=<0.6 0 0>
     }]
    relativeSpaceBone {boneName=spine_05}
   }
   n21 = Anim Dynamics - Bone: dyn_hair_dwn_bk_p_02!(ComponentPose=@n20, Alpha=1, GravityOverride:Vector="0,0,-980", ExternalForce="0,0,0") -> (Pose) {
    angularDampingOverride=0.1
    angularSpringConstant=10
    bAngularSpring=true
    bOverrideAngularDamping=true
    bUseGravityOverride=true
    simulationSpace=RootRelative
    alphaBoolBlend {blendOption=Linear}
    alphaScaleBiasClamp {bClampResult=true}
    boundBone {boneName=dyn_hair_dwn_bk_p_02}
    componentAppliedLinearAccClamp=<3000 3000 3000>
    componentLinearAccScale=<0.1 0.2>
    componentLinearVelScale=<0.1 0.2 0.1>
    gravityOverride {z=-980}
    physicsBodyDefinitions [{
      boundBone {boneName=dyn_hair_dwn_bk_p_02}
      boxExtents=<10 10 10>
      constraintSetup {angularLimitsMax=<3 10 15>, angularLimitsMin=<-3 -33 -15>, angularTarget=<1 0 0>}
      localJointOffset=<0.6 0 0>
     }]
    relativeSpaceBone {boneName=spine_05}
   }
   n22 = Anim Dynamics - Bone: dyn_hair_dwn_bk_Q_02!(ComponentPose=@n21, Alpha=1, GravityOverride:Vector="0,0,-980", ExternalForce="0,0,0") -> (Pose) {
    angularDampingOverride=0.1
    angularSpringConstant=10
    bAngularSpring=true
    bOverrideAngularDamping=true
    bUseGravityOverride=true
    simulationSpace=RootRelative
    alphaBoolBlend {blendOption=Linear}
    alphaScaleBiasClamp {bClampResult=true}
    boundBone {boneName=dyn_hair_dwn_bk_Q_02}
    componentAppliedLinearAccClamp=<3000 3000 3000>
    componentLinearAccScale=<0.1 0.2>
    componentLinearVelScale=<0.1 0.2 0.1>
    gravityOverride {z=-980}
    physicsBodyDefinitions [{
      boundBone {boneName=dyn_hair_dwn_bk_Q_02}
      boxExtents=<10 10 10>
      constraintSetup {angularLimitsMax=<15 10 15>, angularLimitsMin=<-11 -60 -15>, angularTarget=<1 0 0>}
      localJointOffset=<0.6 0 0>
     }]
    relativeSpaceBone {boneName=spine_05}
   }
   "Anim Dynamics - Chain: dyn_hair_dwn_side_R_01 -> dyn_hair_dwn_side_R_02"!(ComponentPose=@n22, Alpha=1, GravityOverride:Vector="0,0,-980", ExternalForce="0,0,0") -> (Pose) {
    angularDampingOverride=0.1
    angularSpringConstant=10
    bAngularSpring=true
    bChain=true
    bOverrideAngularDamping=true
    bUseGravityOverride=true
    simulationSpace=RootRelative
    alphaBoolBlend {blendOption=Linear}
    boundBone {boneName=dyn_hair_dwn_side_R_01}
    chainEnd {boneName=dyn_hair_dwn_side_R_02}
    componentAppliedLinearAccClamp=<3000 3000 3000>
    componentLinearAccScale=<0.1 0.1 0.1>
    componentLinearVelScale=<0.1 0.1 0.1>
    gravityOverride {z=-980}
    physicsBodyDefinitions [
     {
      boundBone {boneName=dyn_hair_dwn_side_R_01}
      boxExtents=<10 10 10>
      constraintSetup {angularLimitsMax=<15 10 15>, angularLimitsMin=<-11 -60 -15>, angularTarget=<0 0 0>}
      localJointOffset=<0.1 0 0>
     }
     {
      boundBone {boneName=dyn_hair_dwn_side_R_02}
      boxExtents=<10 10 10>
      constraintSetup {angularLimitsMax=<0 0 0>, angularLimitsMin=<0 0 0>, angularTarget=<0 0 0>}
      localJointOffset=<4.1 0 0>
     }
    ]
    relativeSpaceBone {boneName=spine_05}
   }
   n23 = "Get hair Rot (up_dwn)"() -> ("hair Rot (up_dwn)":double)
   "Transform (Modify) Bone - Bone: base_hair_dwn_side_R"(Alpha=@n23, Translation="0,0,0", Rotation="50.998,17.389,13.610", Scale="1,1,1") {
    rotationMode=BMM_Additive
    rotationSpace=BCS_BoneSpace
    translationSpace=BCS_BoneSpace
    alphaBoolBlend {blendOption=Linear}
    alphaScaleBiasClamp {bMapRange=true, inRange {max=0, min=-33}, outRange {max=0, min=1}}
    boneToModify {boneName=base_hair_dwn_side_R}
    rotation=<rot 50.998 17.389 13.61>
   }
   n24 = "Transform (Modify) Bone - Bone: Master_bk_hairsh"(Alpha=@n23, Translation="0,0,0", Rotation="74.688,12.133,12.564", Scale="1,1,1") {
    rotationMode=BMM_Additive
    rotationSpace=BCS_BoneSpace
    translationSpace=BCS_BoneSpace
    alphaBoolBlend {blendOption=Linear}
    alphaScaleBiasClamp {bMapRange=true, inRange {max=0, min=-33}, outRange {max=0, min=1}}
    boneToModify {boneName=Master_bk_hairsh}
    rotation=<rot 74.6879 12.133 12.564>
   }
   n25 = "Transform (Modify) Bone - Bone: base_hair_dwn_bk_k"(ComponentPose=@n24, Alpha=@n23, Translation="0,0,0", Rotation="50.998,17.389,13.610", Scale="1,1,1") {
    rotationMode=BMM_Additive
    rotationSpace=BCS_BoneSpace
    translationSpace=BCS_BoneSpace
    alphaBoolBlend {blendOption=Linear}
    alphaScaleBiasClamp {bMapRange=true, inRange {max=0, min=-33}, outRange {max=0, min=1}}
    boneToModify {boneName=base_hair_dwn_bk_k}
    rotation=<rot 50.998 17.389 13.61>
   }
   n26 = "Transform (Modify) Bone - Bone: base_hair_dwn_bk_m"(ComponentPose=@n25, Alpha=@n23, Translation="0,0,0", Rotation="30,-0.028,-0.039", Scale="1,1,1") {
    rotationMode=BMM_Additive
    rotationSpace=BCS_BoneSpace
    translationSpace=BCS_BoneSpace
    alphaBoolBlend {blendOption=Linear}
    alphaScaleBiasClamp {bMapRange=true, inRange {max=0, min=-33}, outRange {max=0, min=1}}
    boneToModify {boneName=base_hair_dwn_bk_m}
    rotation=<rot 30 -0.028 -0.039>
   }
   "Transform (Modify) Bone - Bone: base_hair_dwn_bk_J"(ComponentPose=@n26, Alpha=@n23, Translation="0,0,0", Rotation="83.175,-149.470,-132.307", Scale="1,1,1") {
    rotationMode=BMM_Additive
    rotationSpace=BCS_BoneSpace
    translationSpace=BCS_BoneSpace
    alphaBoolBlend {blendOption=Linear}
    alphaScaleBiasClamp {bMapRange=true, inRange {max=0, min=-33}, outRange {max=0, min=1}}
    boneToModify {boneName=base_hair_dwn_bk_J}
    rotation=<rot 83.175 -149.47 -132.307>
   }
   Output Pose(Result=Use cached pose 'Rban'()) {layerGroup=None}
   n27 = Get body() -> (body:Skeletal Mesh Component Object Reference)
   n28 = "Get hair Rot (up_dwn)"()
   n29 = "Transform (Modify) Bone - Bone: Master_bk_hairsh"(
    ComponentPose=Local To Component(LocalPose=Copy Pose From Mesh(SourceMeshComponent=@n27) {bUseAttachedParent=true})
    Alpha=@n28
    Translation="0,0,0"
    Rotation="64.685,80.146,133.957"
    Scale="1,1,1"
   ) {
    rotationMode=BMM_Additive
    rotationSpace=BCS_BoneSpace
    translationSpace=BCS_BoneSpace
    alphaBoolBlend {blendOption=Linear}
    alphaScaleBiasClamp {bMapRange=true, inRange {max=0, min=-33}, outRange {max=0, min=1}}
    boneToModify {boneName=Master_bk_hairsh}
    rotation=<rot 64.685 80.146 133.9569>
   }
   n30 = "Transform (Modify) Bone - Bone: base_hair_dwn_bk_k"(ComponentPose=@n29, Alpha=@n28, Translation="0,0,0", Rotation="50.998,17.389,13.610", Scale="1,1,1") {
    rotationMode=BMM_Additive
    rotationSpace=BCS_BoneSpace
    translationSpace=BCS_BoneSpace
    alphaBoolBlend {blendOption=Linear}
    alphaScaleBiasClamp {bMapRange=true, inRange {max=0, min=-33}, outRange {max=0, min=1}}
    boneToModify {boneName=base_hair_dwn_bk_k}
    rotation=<rot 50.998 17.389 13.61>
   }
   n31 = "Transform (Modify) Bone - Bone: base_hair_dwn_bk_m"(ComponentPose=@n30, Alpha=@n28, Translation="0,0,0", Rotation="30,-0.028,-0.039", Scale="1,1,1") {
    rotationMode=BMM_Additive
    rotationSpace=BCS_BoneSpace
    translationSpace=BCS_BoneSpace
    alphaBoolBlend {blendOption=Linear}
    alphaScaleBiasClamp {bMapRange=true, inRange {max=0, min=-33}, outRange {max=0, min=1}}
    boneToModify {boneName=base_hair_dwn_bk_m}
    rotation=<rot 30 -0.028 -0.039>
   }
   n32 = "Transform (Modify) Bone - Bone: base_hair_dwn_bk_J"(ComponentPose=@n31, Alpha=@n28, Translation="0,0,0", Rotation="83.175,-149.470,-132.307", Scale="1,1,1") {
    rotationMode=BMM_Additive
    rotationSpace=BCS_BoneSpace
    translationSpace=BCS_BoneSpace
    alphaBoolBlend {blendOption=Linear}
    alphaScaleBiasClamp {bMapRange=true, inRange {max=0, min=-33}, outRange {max=0, min=1}}
    boneToModify {boneName=base_hair_dwn_bk_J}
    rotation=<rot 83.175 -149.47 -132.307>
   }
   n33 = "Transform (Modify) Bone - Bone: base_hair_dwn_side_R"(ComponentPose=@n32, Alpha=@n28, Translation="0,0,0", Rotation="0.000,-16.562,-0.001", Scale="1,1,1") {
    rotationMode=BMM_Additive
    rotationSpace=BCS_BoneSpace
    translationSpace=BCS_BoneSpace
    alphaBoolBlend {blendOption=Linear}
    alphaScaleBiasClamp {bMapRange=true, inRange {max=0, min=-33}, outRange {max=0, min=1}}
    boneToModify {boneName=base_hair_dwn_side_R}
    rotation=<rot 0 -16.562 -0.0011>
   }
   n34 = RigidBody!(
    ComponentPose=@n33
    Alpha=Get hair speed scale() -> (hair speed scale:double)
    OverrideWorldGravity:Vector=Get Gravity Override()
    ExternalForce
    ComponentLinearAccScale:Vector="0.2,0.2,0.2"
    ComponentLinearVelScale:Vector="0.4,0.3,0.3"
   ) -> (Pose) {
    bClampLinearTranslationLimitToRefPose=true
    bOverrideWorldGravity=true
    overridePhysicsAsset=&/M_MED_PsychicAphid_Exo_FaceAcc_Physics
    simulationSpace=BaseBoneSpace
    alphaBoolBlend {blendOption=Linear}
    alphaScaleBiasClamp {bInterpResult=true, interpSpeedIncreasing=2}
    baseBoneRef {boneName=head}
    componentAppliedLinearAccClamp=<3000 3000 3000>
    componentLinearAccScale=<0.4 0.3 0.3>
    componentLinearVelScale=<0.4 0.3 0.3>
    simSpaceSettings {
     maxAngularAcceleration=1000
     maxAngularVelocity=3
     maxLinearAcceleration=700
     maxLinearVelocity=500
     velocityScaleZ=0
     worldAlpha=1
     externalLinearDragV=<0.2 0.2>
    }
   }
   Save cached pose 'Rban'(Pose=Component To Local(ComponentPose=@n34)) {cacheName=Rban}
   n35 = Control Rig!(
    Source=Copy Pose From Mesh(SourceMeshComponent=Get body()) {bUseAttachedParent=true}
    Alpha
    Hair interp while fly:bool=Get Hair fly interp() -> (Hair fly interp:bool)
    hair speed scale:double=Get hair speed scale()
   ) -> (Pose) {
    alphaBoolBlend {blendOption=Linear}
    controlRigAssetReference {blueprintRigClass=&/M_MED_PsychicAphid_Exo_FaceAcc_CtrlRig.M_MED_PsychicAphid_Exo_FaceAcc_CtrlRig_C}
   }
   n36 = Anim Dynamics - Bone: dyn_hair_dwn_fr_a!(ComponentPose=Local To Component(LocalPose=@n35), Alpha=1, ExternalForce="0,0,0") -> (Pose) {
    angularBiasOverride=0.2
    angularDampingOverride=0.99
    angularSpringConstant=14000
    bAngularSpring=true
    bOverrideAngularBias=true
    bOverrideAngularDamping=true
    bUseGravityOverride=true
    gravityScale=0.35
    alphaBoolBlend {blendOption=Linear}
    boundBone {boneName=dyn_hair_dwn_fr_a}
    componentAppliedLinearAccClamp=<200000 200000 200000>
    componentLinearAccScale=<1 1 1>
    componentLinearVelScale=<1 1 1>
    gravityOverride {z=-3000}
    physicsBodyDefinitions [{
      boundBone {boneName=dyn_hair_dwn_fr_a}
      boxExtents=<40 40 40>
      constraintSetup {angularLimitsMax=<7.5 7.5 7.5>, angularLimitsMin=<-7.5 -7.5 -7.5>, angularTarget=<1 0 0>}
      localJointOffset=<1.6649 0.068 0.093>
     }]
    relativeSpaceBone {boneName=head}
   }
   n37 = Anim Dynamics - Bone: dyn_hair_dwn_b!(ComponentPose=@n36, Alpha=1, ExternalForce="0,0,0") -> (Pose) {
    angularBiasOverride=0.2
    angularDampingOverride=0.99
    angularSpringConstant=14000
    bAngularSpring=true
    bOverrideAngularBias=true
    bOverrideAngularDamping=true
    bUseGravityOverride=true
    gravityScale=0.35
    alphaBoolBlend {blendOption=Linear}
    boundBone {boneName=dyn_hair_dwn_b}
    componentAppliedLinearAccClamp=<200000 200000 200000>
    componentLinearAccScale=<1 1 1>
    componentLinearVelScale=<1 1 1>
    gravityOverride {z=-3000}
    physicsBodyDefinitions [{
      boundBone {boneName=dyn_hair_dwn_b}
      boxExtents=<40 40 40>
      constraintSetup {angularLimitsMax=<7.5 7.5 7.5>, angularLimitsMin=<-7.5 -7.5 -7.5>, angularTarget=<1 0 0>}
      localJointOffset=<2.7554 -1.1705 1.105>
     }]
    relativeSpaceBone {boneName=head}
   }
   n38 = Anim Dynamics - Bone: dyn_hair_dwn_c!(ComponentPose=@n37, Alpha=1, ExternalForce="0,0,0") -> (Pose) {
    angularBiasOverride=0.2
    angularDampingOverride=0.99
    angularSpringConstant=14000
    bAngularSpring=true
    bOverrideAngularBias=true
    bOverrideAngularDamping=true
    bUseGravityOverride=true
    gravityScale=0.35
    alphaBoolBlend {blendOption=Linear}
    boundBone {boneName=dyn_hair_dwn_c}
    componentAppliedLinearAccClamp=<200000 200000 200000>
    componentLinearAccScale=<1 1 1>
    componentLinearVelScale=<1 1 1>
    gravityOverride {z=-3000}
    physicsBodyDefinitions [{
      boundBone {boneName=dyn_hair_dwn_c}
      boxExtents=<40 40 40>
      constraintSetup {angularLimitsMax=<7.5 7.5 7.5>, angularLimitsMin=<-7.5 -7.5 -7.5>, angularTarget=<1 0 0>}
      localJointOffset=<2.7554 -1.1705 1.105>
     }]
    relativeSpaceBone {boneName=head}
   }
   n39 = Anim Dynamics - Bone: dyn_hair_dwn_d!(ComponentPose=@n38, Alpha=1, ExternalForce="0,0,0") -> (Pose) {
    angularBiasOverride=0.2
    angularDampingOverride=0.99
    angularSpringConstant=14000
    bAngularSpring=true
    bOverrideAngularBias=true
    bOverrideAngularDamping=true
    bUseGravityOverride=true
    gravityScale=0.35
    alphaBoolBlend {blendOption=Linear}
    boundBone {boneName=dyn_hair_dwn_d}
    componentAppliedLinearAccClamp=<200000 200000 200000>
    componentLinearAccScale=<1 1 1>
    componentLinearVelScale=<1 1 1>
    gravityOverride {z=-3000}
    physicsBodyDefinitions [{
      boundBone {boneName=dyn_hair_dwn_d}
      boxExtents=<40 40 40>
      constraintSetup {angularLimitsMax=<7.5 7.5 7.5>, angularLimitsMin=<-7.5 -7.5 -7.5>, angularTarget=<1 0 0>}
      localJointOffset=<1.495 0.5196 -0.7997>
     }]
    relativeSpaceBone {boneName=head}
   }
   n40 = Anim Dynamics - Bone: dyn_hair_dwn_e!(ComponentPose=@n39, Alpha=1, ExternalForce="0,0,0") -> (Pose) {
    angularBiasOverride=0.2
    angularDampingOverride=0.99
    angularSpringConstant=14000
    bAngularSpring=true
    bOverrideAngularBias=true
    bOverrideAngularDamping=true
    bUseGravityOverride=true
    gravityScale=0.35
    alphaBoolBlend {blendOption=Linear}
    boundBone {boneName=dyn_hair_dwn_e}
    componentAppliedLinearAccClamp=<200000 200000 200000>
    componentLinearAccScale=<1 1 1>
    componentLinearVelScale=<1 1 1>
    gravityOverride {z=-3000}
    physicsBodyDefinitions [{
      boundBone {boneName=dyn_hair_dwn_e}
      boxExtents=<40 40 40>
      constraintSetup {angularLimitsMax=<7.5 7.5 7.5>, angularLimitsMin=<-7.5 -7.5 -7.5>, angularTarget=<1 0 0>}
      localJointOffset=<0.75 0.8534 1.6045>
     }]
    relativeSpaceBone {boneName=head}
   }
   n41 = Anim Dynamics - Bone: dyn_hair_dwn_f!(ComponentPose=@n40, Alpha=1, ExternalForce="0,0,0") -> (Pose) {
    angularBiasOverride=0.2
    angularDampingOverride=0.99
    angularSpringConstant=14000
    bAngularSpring=true
    bOverrideAngularBias=true
    bOverrideAngularDamping=true
    bUseGravityOverride=true
    gravityScale=0.35
    alphaBoolBlend {blendOption=Linear}
    boundBone {boneName=dyn_hair_dwn_f}
    componentAppliedLinearAccClamp=<200000 200000 200000>
    componentLinearAccScale=<1 1 1>
    componentLinearVelScale=<1 1 1>
    gravityOverride {z=-3000}
    physicsBodyDefinitions [{
      boundBone {boneName=dyn_hair_dwn_f}
      boxExtents=<40 40 40>
      constraintSetup {angularLimitsMax=<7.5 7.5 7.5>, angularLimitsMin=<-7.5 -7.5 -7.5>, angularTarget=<1 0 0>}
      localJointOffset=<0.7747 2.7737 1.1391>
     }]
    relativeSpaceBone {boneName=head}
   }
   n42 = Anim Dynamics - Bone: dyn_hair_dwn_g!(ComponentPose=@n41, Alpha=1, ExternalForce="0,0,0") -> (Pose) {
    angularBiasOverride=0.2
    angularDampingOverride=0.99
    angularSpringConstant=14000
    bAngularSpring=true
    bOverrideAngularBias=true
    bOverrideAngularDamping=true
    bUseGravityOverride=true
    gravityScale=0.35
    alphaBoolBlend {blendOption=Linear}
    boundBone {boneName=dyn_hair_dwn_g}
    componentAppliedLinearAccClamp=<200000 200000 200000>
    componentLinearAccScale=<1 1 1>
    componentLinearVelScale=<1 1 1>
    gravityOverride {z=-3000}
    physicsBodyDefinitions [{
      boundBone {boneName=dyn_hair_dwn_g}
      boxExtents=<40 40 40>
      constraintSetup {angularLimitsMax=<7.5 7.5 7.5>, angularLimitsMin=<-7.5 -7.5 -7.5>, angularTarget=<1 0 0>}
      localJointOffset=<0.9493 1.4429 2.2624>
     }]
    relativeSpaceBone {boneName=head}
   }
   n43 = Anim Dynamics - Bone: dyn_hair_dwn_h!(ComponentPose=@n42, Alpha=1, ExternalForce="0,0,0") -> (Pose) {
    angularBiasOverride=0.2
    angularDampingOverride=0.99
    angularSpringConstant=14000
    bAngularSpring=true
    bOverrideAngularBias=true
    bOverrideAngularDamping=true
    bUseGravityOverride=true
    gravityScale=0.35
    alphaBoolBlend {blendOption=Linear}
    boundBone {boneName=dyn_hair_dwn_h}
    componentAppliedLinearAccClamp=<200000 200000 200000>
    componentLinearAccScale=<1 1 1>
    componentLinearVelScale=<1 1 1>
    gravityOverride {z=-3000}
    physicsBodyDefinitions [{
      boundBone {boneName=dyn_hair_dwn_h}
      boxExtents=<40 40 40>
      constraintSetup {angularLimitsMax=<7.5 7.5 7.5>, angularLimitsMin=<-7.5 -7.5 -7.5>, angularTarget=<1 0 0>}
      localJointOffset=<4 0.5 0.5>
     }]
    relativeSpaceBone {boneName=head}
   }
   n44 = Anim Dynamics - Bone: dyn_hair_dwn_i!(ComponentPose=@n43, Alpha=1, ExternalForce="0,0,0") -> (Pose) {
    angularBiasOverride=0.2
    angularDampingOverride=0.99
    angularSpringConstant=14000
    bAngularSpring=true
    bOverrideAngularBias=true
    bOverrideAngularDamping=true
    bUseGravityOverride=true
    gravityScale=0.35
    alphaBoolBlend {blendOption=Linear}
    boundBone {boneName=dyn_hair_dwn_i}
    componentAppliedLinearAccClamp=<200000 200000 200000>
    componentLinearAccScale=<1 1 1>
    componentLinearVelScale=<1 1 1>
    gravityOverride {z=-3000}
    physicsBodyDefinitions [{
      boundBone {boneName=dyn_hair_dwn_i}
      boxExtents=<40 40 40>
      constraintSetup {angularLimitsMax=<7.5 7.5 7.5>, angularLimitsMin=<-7.5 -7.5 -7.5>, angularTarget=<1 0 0>}
      localJointOffset=<2.0441 -0.0305 1.5432>
     }]
    relativeSpaceBone {boneName=head}
   }
   n45 = Anim Dynamics - Bone: dyn_hair_dwn_bk_J_01(ComponentPose=@n44, Alpha=1, ExternalForce="0,0,0") {
    angularBiasOverride=0.2
    angularDampingOverride=0.99
    angularSpringConstant=14000
    bAngularSpring=true
    bOverrideAngularBias=true
    bOverrideAngularDamping=true
    bUseGravityOverride=true
    gravityScale=0.35
    alphaBoolBlend {blendOption=Linear}
    boundBone {boneName=dyn_hair_dwn_bk_J_01}
    componentAppliedLinearAccClamp=<200000 200000 200000>
    componentLinearAccScale=<1 1 1>
    componentLinearVelScale=<1 1 1>
    gravityOverride {z=-3000}
    physicsBodyDefinitions [{
      boundBone {boneName=dyn_hair_dwn_bk_J_01}
      boxExtents=<40 40 40>
      constraintSetup {angularLimitsMax=<7.5 7.5 7.5>, angularLimitsMin=<-7.5 -7.5 -7.5>, angularTarget=<1 0 0>}
      localJointOffset=<2.4801 2.2397 1.6312>
     }]
    relativeSpaceBone {boneName=head}
   }
   n46 = Anim Dynamics - Bone: dyn_hair_dwn_side_S_01!(ComponentPose=@n45, Alpha=1, ExternalForce="0,0,0") -> (Pose) {
    angularBiasOverride=0.2
    angularDampingOverride=0.99
    angularSpringConstant=14000
    bAngularSpring=true
    bOverrideAngularBias=true
    bOverrideAngularDamping=true
    bUseGravityOverride=true
    gravityScale=0.35
    alphaBoolBlend {blendOption=Linear}
    boundBone {boneName=dyn_hair_dwn_side_S_01}
    componentAppliedLinearAccClamp=<200000 200000 200000>
    componentLinearAccScale=<1 1 1>
    componentLinearVelScale=<1 1 1>
    gravityOverride {z=-3000}
    physicsBodyDefinitions [{
      boundBone {boneName=dyn_hair_dwn_side_S_01}
      boxExtents=<40 40 40>
      constraintSetup {angularLimitsMax=<7.5 7.5 7.5>, angularLimitsMin=<-7.5 -7.5 -7.5>, angularTarget=<1 0 0>}
      localJointOffset=<0.2587 -0.6529 1.3511>
     }]
    relativeSpaceBone {boneName=head}
   }
   n47 = Anim Dynamics - Bone: dyn_hair_up_fr_a!(ComponentPose=@n46, Alpha=1, ExternalForce="0,0,0") -> (Pose) {
    angularBiasOverride=0.2
    angularDampingOverride=0.99
    angularSpringConstant=14000
    bAngularSpring=true
    bOverrideAngularBias=true
    bOverrideAngularDamping=true
    bUseGravityOverride=true
    gravityScale=0.35
    alphaBoolBlend {blendOption=Linear}
    boundBone {boneName=dyn_hair_up_fr_a}
    componentAppliedLinearAccClamp=<200000 200000 200000>
    componentLinearAccScale=<1 1 1>
    componentLinearVelScale=<1 1 1>
    gravityOverride {z=-3000}
    physicsBodyDefinitions [{
      boundBone {boneName=dyn_hair_up_fr_a}
      boxExtents=<40 40 40>
      constraintSetup {angularLimitsMax=<7.5 7.5 7.5>, angularLimitsMin=<-7.5 -7.5 -7.5>, angularTarget=<1 0 0>}
      localJointOffset=<0.8545 1.0907 -0.8031>
     }]
    relativeSpaceBone {boneName=head}
   }
   n48 = Anim Dynamics - Bone: dyn_hair_up_m_b!(ComponentPose=@n47, Alpha=1, ExternalForce="0,0,0") -> (Pose) {
    angularBiasOverride=0.2
    angularDampingOverride=0.99
    angularSpringConstant=14000
    bAngularSpring=true
    bOverrideAngularBias=true
    bOverrideAngularDamping=true
    bUseGravityOverride=true
    gravityScale=0.35
    alphaBoolBlend {blendOption=Linear}
    boundBone {boneName=dyn_hair_up_m_b}
    componentAppliedLinearAccClamp=<200000 200000 200000>
    componentLinearAccScale=<1 1 1>
    componentLinearVelScale=<1 1 1>
    gravityOverride {z=-3000}
    physicsBodyDefinitions [{
      boundBone {boneName=dyn_hair_up_m_b}
      boxExtents=<40 40 40>
      constraintSetup {angularLimitsMax=<7.5 7.5 7.5>, angularLimitsMin=<-7.5 -7.5 -7.5>, angularTarget=<1 0 0>}
      localJointOffset=<1.9715 0.7465 0.7959>
     }]
    relativeSpaceBone {boneName=head}
   }
   n49 = Anim Dynamics - Bone: dyn_hair_up_m_c!(ComponentPose=@n48, Alpha=1, ExternalForce="0,0,0") -> (Pose) {
    angularBiasOverride=0.2
    angularDampingOverride=0.99
    angularSpringConstant=14000
    bAngularSpring=true
    bOverrideAngularBias=true
    bOverrideAngularDamping=true
    bUseGravityOverride=true
    gravityScale=0.35
    alphaBoolBlend {blendOption=Linear}
    boundBone {boneName=dyn_hair_up_m_c}
    componentAppliedLinearAccClamp=<200000 200000 200000>
    componentLinearAccScale=<1 1 1>
    componentLinearVelScale=<1 1 1>
    gravityOverride {z=-3000}
    physicsBodyDefinitions [{
      boundBone {boneName=dyn_hair_up_m_c}
      boxExtents=<40 40 40>
      constraintSetup {angularLimitsMax=<7.5 7.5 7.5>, angularLimitsMin=<-7.5 -7.5 -7.5>, angularTarget=<1 0 0>}
      localJointOffset=<2.2127 0.1441 2.9271>
     }]
    relativeSpaceBone {boneName=head}
   }
   n50 = Anim Dynamics - Bone: dyn_hair_dwn_bk_L!(ComponentPose=@n49, Alpha=1, ExternalForce="0,0,0") -> (Pose) {
    angularBiasOverride=0.2
    angularDampingOverride=0.99
    angularSpringConstant=14000
    bAngularSpring=true
    bOverrideAngularBias=true
    bOverrideAngularDamping=true
    bUseGravityOverride=true
    gravityScale=0.35
    alphaBoolBlend {blendOption=Linear}
    boundBone {boneName=dyn_hair_dwn_bk_L}
    componentAppliedLinearAccClamp=<200000 200000 200000>
    componentLinearAccScale=<1 1 1>
    componentLinearVelScale=<1 1 1>
    gravityOverride {z=-3000}
    physicsBodyDefinitions [{
      boundBone {boneName=dyn_hair_dwn_bk_L}
      boxExtents=<40 40 40>
      constraintSetup {angularLimitsMax=<7.5 7.5 7.5>, angularLimitsMin=<-7.5 -7.5 -7.5>, angularTarget=<1 0 0>}
      localJointOffset=<3 2 0>
     }]
    relativeSpaceBone {boneName=head}
   }
   n51 = Anim Dynamics - Bone: dyn_hair_dwn_bk_k_02!(ComponentPose=@n50, Alpha=1, ExternalForce="0,0,0") -> (Pose) {
    angularBiasOverride=0.2
    angularDampingOverride=0.99
    angularSpringConstant=14000
    bAngularSpring=true
    bOverrideAngularBias=true
    bOverrideAngularDamping=true
    bUseGravityOverride=true
    gravityScale=0.35
    alphaBoolBlend {blendOption=Linear}
    boundBone {boneName=dyn_hair_dwn_bk_k_02}
    componentAppliedLinearAccClamp=<200000 200000 200000>
    componentLinearAccScale=<1 1 1>
    componentLinearVelScale=<1 1 1>
    gravityOverride {z=-3000}
    physicsBodyDefinitions [{
      boundBone {boneName=dyn_hair_dwn_bk_k_02}
      boxExtents=<40 40 40>
      constraintSetup {angularLimitsMax=<7.5 7.5 7.5>, angularLimitsMin=<-7.5 -7.5 -7.5>, angularTarget=<1 0 0>}
      localJointOffset=<2 1.3835 2.143>
     }]
    relativeSpaceBone {boneName=head}
   }
   n52 = Anim Dynamics - Bone: dyn_hair_dwn_bk_m_02!(ComponentPose=@n51, Alpha=1, ExternalForce="0,0,0") -> (Pose) {
    angularBiasOverride=0.2
    angularDampingOverride=0.99
    angularSpringConstant=14000
    bAngularSpring=true
    bOverrideAngularBias=true
    bOverrideAngularDamping=true
    bUseGravityOverride=true
    gravityScale=0.35
    alphaBoolBlend {blendOption=Linear}
    boundBone {boneName=dyn_hair_dwn_bk_m_02}
    componentAppliedLinearAccClamp=<200000 200000 200000>
    componentLinearAccScale=<1 1 1>
    componentLinearVelScale=<1 1 1>
    gravityOverride {z=-3000}
    physicsBodyDefinitions [{
      boundBone {boneName=dyn_hair_dwn_bk_m_02}
      boxExtents=<40 40 40>
      constraintSetup {angularLimitsMax=<7.5 7.5 7.5>, angularLimitsMin=<-7.5 -7.5 -7.5>, angularTarget=<1 0 0>}
      localJointOffset=<0.4827 2.8969 1.037>
     }]
    relativeSpaceBone {boneName=head}
   }
   n53 = Anim Dynamics - Bone: dyn_hair_dwn_bk_n!(ComponentPose=@n52, Alpha=1, ExternalForce="0,0,0") -> (Pose) {
    angularBiasOverride=0.2
    angularDampingOverride=0.99
    angularSpringConstant=14000
    bAngularSpring=true
    bOverrideAngularBias=true
    bOverrideAngularDamping=true
    bUseGravityOverride=true
    gravityScale=0.35
    alphaBoolBlend {blendOption=Linear}
    boundBone {boneName=dyn_hair_dwn_bk_n}
    componentAppliedLinearAccClamp=<200000 200000 200000>
    componentLinearAccScale=<1 1 1>
    componentLinearVelScale=<1 1 1>
    gravityOverride {z=-3000}
    physicsBodyDefinitions [{
      boundBone {boneName=dyn_hair_dwn_bk_n}
      boxExtents=<40 40 40>
      constraintSetup {angularLimitsMax=<7.5 7.5 7.5>, angularLimitsMin=<-7.5 -7.5 -7.5>, angularTarget=<1 0 0>}
      localJointOffset=<0.9506 1.0018 -0.4806>
     }]
    relativeSpaceBone {boneName=head}
   }
   n54 = Anim Dynamics - Bone: dyn_hair_dwn_bk_O!(ComponentPose=@n53, Alpha=1, ExternalForce="0,0,0") -> (Pose) {
    angularBiasOverride=0.2
    angularDampingOverride=0.99
    angularSpringConstant=14000
    bAngularSpring=true
    bOverrideAngularBias=true
    bOverrideAngularDamping=true
    bUseGravityOverride=true
    gravityScale=0.35
    alphaBoolBlend {blendOption=Linear}
    boundBone {boneName=dyn_hair_dwn_bk_O}
    componentAppliedLinearAccClamp=<200000 200000 200000>
    componentLinearAccScale=<1 1 1>
    componentLinearVelScale=<1 1 1>
    gravityOverride {z=-3000}
    physicsBodyDefinitions [{
      boundBone {boneName=dyn_hair_dwn_bk_O}
      boxExtents=<40 40 40>
      constraintSetup {angularLimitsMax=<7.5 7.5 7.5>, angularLimitsMin=<-7.5 -7.5 -7.5>, angularTarget=<1 0 0>}
      localJointOffset=<0.3051 1.3727 1.4087>
     }]
    relativeSpaceBone {boneName=head}
   }
   n55 = Anim Dynamics - Bone: dyn_hair_dwn_bk_p_02!(ComponentPose=@n54, Alpha=1, ExternalForce="0,0,0") -> (Pose) {
    angularBiasOverride=0.2
    angularDampingOverride=0.99
    angularSpringConstant=14000
    bAngularSpring=true
    bOverrideAngularBias=true
    bOverrideAngularDamping=true
    bUseGravityOverride=true
    gravityScale=0.35
    alphaBoolBlend {blendOption=Linear}
    boundBone {boneName=dyn_hair_dwn_bk_p_02}
    componentAppliedLinearAccClamp=<200000 200000 200000>
    componentLinearAccScale=<1 1 1>
    componentLinearVelScale=<1 1 1>
    gravityOverride {z=-3000}
    physicsBodyDefinitions [{
      boundBone {boneName=dyn_hair_dwn_bk_p_02}
      boxExtents=<40 40 40>
      constraintSetup {angularLimitsMax=<7.5 7.5 7.5>, angularLimitsMin=<-7.5 -7.5 -7.5>, angularTarget=<1 0 0>}
      localJointOffset=<2.7787 -0.9688 0.1697>
     }]
    relativeSpaceBone {boneName=head}
   }
   n56 = Anim Dynamics - Bone: dyn_hair_dwn_bk_Q_02!(ComponentPose=@n55, Alpha=1, ExternalForce="0,0,0") -> (Pose) {
    angularBiasOverride=0.2
    angularDampingOverride=0.99
    angularSpringConstant=14000
    bAngularSpring=true
    bOverrideAngularBias=true
    bOverrideAngularDamping=true
    bUseGravityOverride=true
    gravityScale=0.35
    alphaBoolBlend {blendOption=Linear}
    boundBone {boneName=dyn_hair_dwn_bk_Q_02}
    componentAppliedLinearAccClamp=<200000 200000 200000>
    componentLinearAccScale=<1 1 1>
    componentLinearVelScale=<1 1 1>
    gravityOverride {z=-3000}
    physicsBodyDefinitions [{
      boundBone {boneName=dyn_hair_dwn_bk_Q_02}
      boxExtents=<40 40 40>
      constraintSetup {angularLimitsMax=<7.5 7.5 7.5>, angularLimitsMin=<-7.5 -7.5 -7.5>, angularTarget=<1 0 0>}
      localJointOffset=<2.7787 -0.9688 0.1697>
     }]
    relativeSpaceBone {boneName=head}
   }
   n57 = Anim Dynamics - Bone: dyn_hair_dwn_side_R_02(ComponentPose=@n56, Alpha=1, ExternalForce="0,0,0") {
    angularBiasOverride=0.2
    angularDampingOverride=0.99
    angularSpringConstant=14000
    bAngularSpring=true
    bOverrideAngularBias=true
    bOverrideAngularDamping=true
    bUseGravityOverride=true
    gravityScale=0.35
    alphaBoolBlend {blendOption=Linear}
    boundBone {boneName=dyn_hair_dwn_side_R_02}
    componentAppliedLinearAccClamp=<200000 200000 200000>
    componentLinearAccScale=<1 1 1>
    componentLinearVelScale=<1 1 1>
    gravityOverride {z=-3000}
    physicsBodyDefinitions [{
      boundBone {boneName=dyn_hair_dwn_side_R_02}
      boxExtents=<40 40 40>
      constraintSetup {angularLimitsMax=<7.5 7.5 7.5>, angularLimitsMin=<-7.5 -7.5 -7.5>, angularTarget=<1 0 0>}
      localJointOffset=<1.7893 1.4003 1.2315>
     }]
    relativeSpaceBone {boneName=head}
   }
   Save cached pose 'Animdyn'(Pose=Component To Local(ComponentPose=@n57)) {cacheName=Animdyn}
   Use cached pose 'Animdyn'()
  ]}
 EventGraph {nodes [
   Event BlueprintUpdateAnimation()
   > Sequence()
   > n1 = Gravity Override State Function DEPRECATED(CharacterPartAnimInstance=Self-Reference(), GameTimeInSeconds=GetGameTimeInSeconds(), GravityOverrideParams=Get Gravity Override Params())
   > Set Gravity Override(Gravity Override=@n1)
   n2 = GetDeltaTransformFromRefPose(self=GetOwnerSkeletalMeshComponent(), BoneName=head, BaseName=neck_01)
   > "Set hair Rot (up_dwn)"("hair Rot (up_dwn)":double=BreakRotator(InRot=BreakTransform(InTransform=@n2).Rotation).Yaw) -> (Output_Get:double)
   > Set Hair fly interp(Hair fly interp:bool=OR Boolean!(A=Get bIsSkydiveFloating(), B=Get bIsParachuteOpened(), C:bool=Get bIsSkyDiving()) -> (ReturnValue)) -> (Output_Get:bool)
   > n3 = Branch(Condition=Get bIsSkydiveFloating())
   > Set hair speed scale(hair speed scale:double=5) -> (Output_Get:double)
   n4 = Branch(execute=@n3.else, Condition=Get bIsParachuteOpened())
   > Set hair speed scale(hair speed scale=5)
   n5 = Branch(execute=@n4.else, Condition=Get bIsSkyDiving())
   > Set hair speed scale(hair speed scale=8)
   Set hair speed scale(execute=@n5.else, hair speed scale=0)
   Event BlueprintInitializeAnimation()
   > Set body(body:Skeletal Mesh Component Object Reference=GetOwnerSkeletalMeshComponent()) -> (Output_Get:Skeletal Mesh Component Object Reference)
  ]}
]
variables [
 body {type=SkeletalMeshComponent}
 Gravity Override {type=Vector}
 Gravity Override Params {type=GravityOverrideParamsStruct}
 Hair fly interp {type=boolean}
 "hair Rot (up_dwn)" {type=number}
 hair speed scale {type=number}
 Head Fwd Bwd Rot {type=number}
]
